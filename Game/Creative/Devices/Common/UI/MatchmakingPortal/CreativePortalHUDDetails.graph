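# HUD widget showing a Creative island portal's details: name, creator, link code, population, featured label, description and tags.
# InitFromObject stores the BGA_IslandPortal and binds its update events; setters fill text and toggle visibility, including privacy-based link code hiding.
# Population also collapses for hub-return or name-override portals, and tags are rebuilt as Cr_ItemTag widgets.
# UpdateVisBasedOnDistance toggles the description by flags, triggering reveal and outline-material sheen animations on change.
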WidgetBlueprint /Game/Creative/Devices/Common/UI/MatchmakingPortal/CreativePortalHUDDetails
animations [
 Intro {
  binding_count=3
  length_seconds=0.75
  object=&self:Intro
  sequence {bindings [
    Text_PlayerName {
     id=#1
     possessed_class=CommonTextBlock
     tracks [
      {
       type=MovieScene2DTransformTrack
       sections [{
         type=MovieScene2DTransformSection
         end_frame=5
         start_frame=0
         channels [
          Translation.X {type=MovieSceneScriptingFloatChannel}
          Translation.Y {type=MovieSceneScriptingFloatChannel}
          Angle {type=MovieSceneScriptingFloatChannel}
          Scale.X {type=MovieSceneScriptingFloatChannel, keys [[0,0.7056],[2,1.1],[3,0.95],[5,1]]}
          Scale.Y {type=MovieSceneScriptingFloatChannel, keys [[0,0.7056],[2,1.1],[3,0.95],[5,1]]}
          Shear.X {type=MovieSceneScriptingFloatChannel}
          Shear.Y {type=MovieSceneScriptingFloatChannel}
         ]
        }]
      }
      {
       type=MovieSceneFloatTrack
       display_name=Render Opacity
       sections [{type=MovieSceneFloatSection, end_frame=5, start_frame=0, channels [None {type=MovieSceneScriptingFloatChannel, keys [[0,0],[2,1],[5,1]]}]}]
      }
     ]
    }
    PopulationBox {
     id=#2
     possessed_class=HorizontalBox
     tracks [
      {
       type=MovieScene2DTransformTrack
       sections [{
         type=MovieScene2DTransformSection
         end_frame=15
         start_frame=8
         channels [
          Translation.X {type=MovieSceneScriptingFloatChannel, const [0,4]}
          Translation.Y {type=MovieSceneScriptingFloatChannel, keys [[8,-9],[11,3.6452],[13,-1.5518],[15,0]]}
          Angle {type=MovieSceneScriptingFloatChannel}
          Scale.X {type=MovieSceneScriptingFloatChannel}
          Scale.Y {type=MovieSceneScriptingFloatChannel}
          Shear.X {type=MovieSceneScriptingFloatChannel}
          Shear.Y {type=MovieSceneScriptingFloatChannel}
         ]
        }]
      }
      {
       type=MovieSceneFloatTrack
       display_name=Render Opacity
       sections [{type=MovieSceneFloatSection, end_frame=10, start_frame=0, channels [None {type=MovieSceneScriptingFloatChannel, keys [[0,0],[8,0],[10,1]]}]}]
      }
     ]
    }
    Text_IslandName {
     id=#3
     possessed_class=CommonTextBlock
     tracks [
      {
       type=MovieScene2DTransformTrack
       sections [{
         type=MovieScene2DTransformSection
         end_frame=15
         start_frame=0
         channels [
          Translation.X {type=MovieSceneScriptingFloatChannel, const [0,6]}
          Translation.Y {type=MovieSceneScriptingFloatChannel, keys [[0,0],[5,-9],[8,1.4508],[10,-0.9672],[12,0],[15,0]]}
          Angle {type=MovieSceneScriptingFloatChannel}
          Scale.X {type=MovieSceneScriptingFloatChannel}
          Scale.Y {type=MovieSceneScriptingFloatChannel}
          Shear.X {type=MovieSceneScriptingFloatChannel}
          Shear.Y {type=MovieSceneScriptingFloatChannel}
         ]
        }]
      }
      {
       type=MovieSceneFloatTrack
       display_name=Render Opacity
       sections [{
         type=MovieSceneFloatSection
         end_frame=15
         start_frame=0
         channels [None {type=MovieSceneScriptingFloatChannel, keys [[0,0],[5,0],[7,1],[15,1]]}]
        }]
      }
     ]
    }
   ]}
 }
 intro2 {
  binding_count=3
  length_seconds=0.15
  object=&self:intro2
  sequence {bindings [
    Text_PlayerName {
     id=#4
     possessed_class=CommonTextBlock
     tracks [
      {
       type=MovieScene2DTransformTrack
       sections [{
         type=MovieScene2DTransformSection
         end_frame=3
         start_frame=0
         channels [
          Translation.X {type=MovieSceneScriptingFloatChannel, keys [[0,-48],[3,0]]}
          Translation.Y {type=MovieSceneScriptingFloatChannel, const [0,2]}
          Angle {type=MovieSceneScriptingFloatChannel}
          Scale.X {type=MovieSceneScriptingFloatChannel}
          Scale.Y {type=MovieSceneScriptingFloatChannel}
          Shear.X {type=MovieSceneScriptingFloatChannel}
          Shear.Y {type=MovieSceneScriptingFloatChannel}
         ]
        }]
      }
      {
       type=MovieSceneFloatTrack
       display_name=Render Opacity
       sections [{type=MovieSceneFloatSection, end_frame=3, start_frame=0, channels [None {type=MovieSceneScriptingFloatChannel, keys [[0,0],[3,1]]}]}]
      }
     ]
    }
    PopulationBox {
     id=#5
     possessed_class=HorizontalBox
     tracks [
      {
       type=MovieScene2DTransformTrack
       sections [{
         type=MovieScene2DTransformSection
         end_frame=3
         start_frame=0
         channels [
          Translation.X {type=MovieSceneScriptingFloatChannel, keys [[0,-50],[3,0]]}
          Translation.Y {type=MovieSceneScriptingFloatChannel, const [0,2]}
          Angle {type=MovieSceneScriptingFloatChannel}
          Scale.X {type=MovieSceneScriptingFloatChannel}
          Scale.Y {type=MovieSceneScriptingFloatChannel}
          Shear.X {type=MovieSceneScriptingFloatChannel}
          Shear.Y {type=MovieSceneScriptingFloatChannel}
         ]
        }]
      }
      {
       type=MovieSceneFloatTrack
       display_name=Render Opacity
       sections [{type=MovieSceneFloatSection, end_frame=3, start_frame=0, channels [None {type=MovieSceneScriptingFloatChannel, keys [[0,0],[3,1]]}]}]
      }
     ]
    }
    Text_IslandName {
     id=#6
     possessed_class=CommonTextBlock
     tracks [
      {
       type=MovieScene2DTransformTrack
       sections [{
         type=MovieScene2DTransformSection
         end_frame=3
         start_frame=0
         channels [
          Translation.X {type=MovieSceneScriptingFloatChannel, keys [[0,50],[3,0]]}
          Translation.Y {type=MovieSceneScriptingFloatChannel, const [0,2]}
          Angle {type=MovieSceneScriptingFloatChannel}
          Scale.X {type=MovieSceneScriptingFloatChannel}
          Scale.Y {type=MovieSceneScriptingFloatChannel}
          Shear.X {type=MovieSceneScriptingFloatChannel}
          Shear.Y {type=MovieSceneScriptingFloatChannel}
         ]
        }]
      }
      {
       type=MovieSceneFloatTrack
       display_name=Render Opacity
       sections [{type=MovieSceneFloatSection, end_frame=3, start_frame=0, channels [None {type=MovieSceneScriptingFloatChannel, keys [[0,0],[3,1]]}]}]
      }
     ]
    }
   ]}
 }
 DescriptionReveal {
  binding_count=2
  length_seconds=0.25
  object=&self:DescriptionReveal
  sequence {bindings [
    CommonTextBlock_1 {
     id=#7
     possessed_class=CommonTextBlock
     tracks [
      {
       type=MovieScene2DTransformTrack
       sections [{
         type=MovieScene2DTransformSection
         end_frame=5
         start_frame=0
         channels [
          Translation.X {type=MovieSceneScriptingFloatChannel}
          Translation.Y {type=MovieSceneScriptingFloatChannel}
          Angle {type=MovieSceneScriptingFloatChannel}
          Scale.X {type=MovieSceneScriptingFloatChannel, keys [[0,1.15],[4,1],[5,1]]}
          Scale.Y {type=MovieSceneScriptingFloatChannel, keys [[0,1.15],[4,1],[5,1]]}
          Shear.X {type=MovieSceneScriptingFloatChannel}
          Shear.Y {type=MovieSceneScriptingFloatChannel}
         ]
        }]
      }
      {
       type=MovieSceneFloatTrack
       display_name=Render Opacity
       sections [{type=MovieSceneFloatSection, end_frame=2, start_frame=0, channels [None {type=MovieSceneScriptingFloatChannel, keys [[0,0],[2,1]]}]}]
      }
     ]
    }
    Text_Description {
     id=#8
     possessed_class=CommonTextBlock
     tracks [
      {
       type=MovieScene2DTransformTrack
       sections [{
         type=MovieScene2DTransformSection
         end_frame=5
         start_frame=0
         channels [
          Translation.X {type=MovieSceneScriptingFloatChannel}
          Translation.Y {type=MovieSceneScriptingFloatChannel}
          Angle {type=MovieSceneScriptingFloatChannel}
          Scale.X {type=MovieSceneScriptingFloatChannel, keys [[0,1.2424],[5,1]]}
          Scale.Y {type=MovieSceneScriptingFloatChannel, keys [[0,1.2424],[5,1]]}
          Shear.X {type=MovieSceneScriptingFloatChannel}
          Shear.Y {type=MovieSceneScriptingFloatChannel}
         ]
        }]
      }
      {
       type=MovieSceneFloatTrack
       display_name=Render Opacity
       sections [{type=MovieSceneFloatSection, end_frame=5, start_frame=0, channels [None {type=MovieSceneScriptingFloatChannel, keys [[0,0],[3,1],[5,1]]}]}]
      }
     ]
    }
   ]}
 }
 VisibilityReveal {
  binding_count=2
  length_seconds=0.45
  object=&self:VisibilityReveal
  sequence {bindings [
    Text_PlayerName {
     id=#9
     possessed_class=CommonTextBlock
     tracks [
      {
       type=MovieSceneWidgetMaterialTrack
       display_name=Font.Outline Settings.OutlineMaterial
       sections [{type=MovieSceneParameterSection, end_frame=9, start_frame=0, channels [Position {type=MovieSceneScriptingFloatChannel, keys [[0,0],[9,1]]}]}]
      }
      {
       type=MovieSceneFloatTrack
       display_name=Render Opacity
       sections [{
         type=MovieSceneFloatSection
         end_frame=9
         start_frame=0
         channels [None {type=MovieSceneScriptingFloatChannel, keys [[0,0],[5,1],[9.0003,1]]}]
        }]
      }
      {
       type=MovieScene2DTransformTrack
       sections [{
         type=MovieScene2DTransformSection
         end_frame=9
         start_frame=0
         channels [
          Translation.X {type=MovieSceneScriptingFloatChannel}
          Translation.Y {type=MovieSceneScriptingFloatChannel}
          Angle {type=MovieSceneScriptingFloatChannel}
          Scale.X {type=MovieSceneScriptingFloatChannel, keys [[0,0.6841],[5,1.1],[7,0.9646],[9,1]]}
          Scale.Y {type=MovieSceneScriptingFloatChannel, const [1,4]}
          Shear.X {type=MovieSceneScriptingFloatChannel}
          Shear.Y {type=MovieSceneScriptingFloatChannel}
         ]
        }]
      }
     ]
    }
    Text_FeaturedPlayerName {
     id=#10
     possessed_class=CommonTextBlock
     tracks [
      {
       type=MovieSceneWidgetMaterialTrack
       display_name=Font.Outline Settings.OutlineMaterial
       sections [{
         type=MovieSceneParameterSection
         end_frame=9
         start_frame=0
         channels [Position {type=MovieSceneScriptingFloatChannel, keys [[0,0],[9.0003,1]]}]
        }]
      }
      {
       type=MovieScene2DTransformTrack
       sections [{
         type=MovieScene2DTransformSection
         end_frame=9
         start_frame=0
         channels [
          Translation.X {type=MovieSceneScriptingFloatChannel}
          Translation.Y {type=MovieSceneScriptingFloatChannel}
          Angle {type=MovieSceneScriptingFloatChannel}
          Scale.X {type=MovieSceneScriptingFloatChannel, keys [[0,0.7],[5,1.1],[7,0.95],[9,1]]}
          Scale.Y {type=MovieSceneScriptingFloatChannel, const [1,4]}
          Shear.X {type=MovieSceneScriptingFloatChannel}
          Shear.Y {type=MovieSceneScriptingFloatChannel}
         ]
        }]
      }
      {
       type=MovieSceneFloatTrack
       display_name=Render Opacity
       sections [{type=MovieSceneFloatSection, end_frame=9, start_frame=0, channels [None {type=MovieSceneScriptingFloatChannel, keys [[0,0],[5,1],[9,1]]}]}]
      }
     ]
    }
   ]}
 }
 SheenNames {
  binding_count=2
  length_seconds=0.45
  object=&self:SheenNames
  sequence {bindings [
    Text_FeaturedPlayerName {
     id=#11
     possessed_class=CommonTextBlock
     tracks [{
       type=MovieSceneWidgetMaterialTrack
       display_name=Font.Outline Settings.OutlineMaterial
       sections [{type=MovieSceneParameterSection, end_frame=9, start_frame=0, channels [Position {type=MovieSceneScriptingFloatChannel, keys [[0,0],[9,1]]}]}]
      }]
    }
    Text_PlayerName {
     id=#12
     possessed_class=CommonTextBlock
     tracks [{
       type=MovieSceneWidgetMaterialTrack
       display_name=Font.Outline Settings.OutlineMaterial
       sections [{type=MovieSceneParameterSection, end_frame=9, start_frame=0, channels [Position {type=MovieSceneScriptingFloatChannel, keys [[0,0],[9,1]]}]}]
      }]
    }
   ]}
 }
 Intro3 {
  binding_count=1
  length_seconds=0.25
  object=&self:Intro3
  sequence {bindings [CreativeLogoWatermark {
     id=#13
     possessed_class=Image
     tracks [
      {
       type=MovieScene2DTransformTrack
       sections [{
         type=MovieScene2DTransformSection
         end_frame=5
         start_frame=0
         channels [
          Translation.X {type=MovieSceneScriptingFloatChannel}
          Translation.Y {type=MovieSceneScriptingFloatChannel}
          Angle {type=MovieSceneScriptingFloatChannel, keys [[0,8],[3,-18.7729],[5,-15]]}
          Scale.X {type=MovieSceneScriptingFloatChannel, keys [[0,2],[3,0.7866],[5,1]]}
          Scale.Y {type=MovieSceneScriptingFloatChannel, keys [[0,2],[3,0.7866],[5,1]]}
          Shear.X {type=MovieSceneScriptingFloatChannel}
          Shear.Y {type=MovieSceneScriptingFloatChannel}
         ]
        }]
      }
      {
       type=MovieSceneFloatTrack
       display_name=Render Opacity
       sections [{type=MovieSceneFloatSection, end_frame=5, start_frame=0, channels [None {type=MovieSceneScriptingFloatChannel, keys [[0,0],[5,1]]}]}]
      }
     ]
    }]}
 }
]
graphs [
 EventGraph {nodes [
   n1 = AND Boolean(
    A=Get IsMyPortal!(self:BGA Island Portal Object Reference=Get MyBGAIslandPortal() -> (MyBGAIslandPortal:BGA Island Portal Object Reference)) -> (IsMyPortal:bool)
    B=IsCreativeModeAccessLimited!(self:Fort Global UIContext Object Reference=Get FortGlobalUIContext() -> (ReturnValue:Fort Global UIContext Object Reference) {is=title K2Node_GetSubsystem}) -> (ReturnValue:bool)
   )
   n2 = Get LimitedAccessText() -> (LimitedAccessText:Common Text Object Reference)
   SetVisibility(self=@n2, InVisibility=Select(Option 0=Collapsed, Option 1=SelfHitTestInvisible, Index=@n1))
   > "SetText (Text)"(self=@n2, InText=Format Text!(Format="LOCTEXT(\"[{x}]\")", x:Text=MakeLiteralText(Value="LOCTEXT(\"LIMITED ACCESS\")")) -> (Result))
   n3 = OnNewVisibilityThresholdChanged!() -> (OutputDelegate, then, bNewVis:bool=~) {is=title K2Node_CustomEvent}
   > Branch(Condition=@n3.bNewVis)
   > PlayAnimation(InAnimation=Get DescriptionReveal() -> (DescriptionReveal:Widget Animation Object Reference))
   > PlayAnimation(InAnimation=Get SheenNames() -> (SheenNames:Widget Animation Object Reference))
   n4 = Event InitFromObject!() -> (OutputDelegate, then, InitObject:Object Reference) {is=title K2Node_Event}
   > n5 = Cast To BGA_IslandPortal!(execute, Object=@n4.InitObject) -> (then, CastFailed, AsBGA Island Portal:BGA Island Portal Object Reference) {is=title K2Node_DynamicCast}
   > n6 = Set MyBGAIslandPortal(MyBGAIslandPortal:BGA Island Portal Object Reference=@n5) -> (Output_Get:BGA Island Portal Object Reference)
   > SetupPortalEventBinding!(execute, self, InPortal:BGA Island Portal Object Reference=@n6, Bind:bool=true) -> (then)
   n7 = Get MyBGAIslandPortal()
   n8 = TextIsEmpty(InText=Get OverrideName!(self:BGA Island Portal Object Reference=@n7) -> (OverrideName:Text))
   > n9 = Branch(Condition=OR Boolean(A=Get bReturnToCreativeHub!(self:Fort Athena Creative Portal Object Reference=@n7) -> (bReturnToCreativeHub:bool), B=NOT Boolean(A=@n8)))
   > SetVisibility(self=Get PopulationBox() -> (PopulationBox:Horizontal Box Object Reference), InVisibility=Collapsed)
   SetVisibility(
    execute=@n9.else
    self=[Get IslandNameBox() -> (IslandNameBox:Vertical Box Object Reference), Get PopulationBox()]
    InVisibility=Select(Option 0=SelfHitTestInvisible, Option 1=Collapsed, Index=@n1)
   )
   Event PlayIntro() {is=title K2Node_Event}
   > PlayAnimationForward(InAnimation=Get Intro3() -> (Intro3:Widget Animation Object Reference))
   Get intro2() -> (intro2:Widget Animation Object Reference)
  ]}
 GetIslandDescriptionTags {nodes [
   GetIslandDescriptionTags() {is=title K2Node_FunctionEntry}
   > n1 = Get!(execute:Exec) -> (then:Exec, else:Exec, MyBGAIslandPortal:BGA Island Portal Object Reference)
   > Return Node!(
    execute
    DescriptionTags:Array of Creative Island Description Tag Structures=GetIslandDescriptionTags!(self:Fort Athena Creative Portal Object Reference=@n1) -> (ReturnValue:Array of Creative Island Description Tag Structures) {is=title K2Node_CallFunction}
   )
   Return Node(execute=@n1.else)
  ]}
 PopulateTags {nodes [
   n1 = PopulateTags!() -> (then, TagArray:Array of Creative Island Description Tag Structures) {is=title K2Node_FunctionEntry}
   > ClearChildren(self=Get WrapBox_Tags() -> (WrapBox_Tags:Wrap Box Object Reference))
   > n2 = For Each Loop(Array:Array of Creative Island Description Tag Structures=@n1) -> (Array Element:Creative Island Description Tag Structure)
   n3 = Break Creative Island Description Tag!(CreativeIslandDescriptionTag:"Creative Island Description Tag Structure (by ref)"=@n2) -> (Name:String, DisplayText:Text) {is=title K2Node_BreakStruct}
   > n4 = Branch(Condition=TextIsEmpty(InText=@n3.DisplayText))
   n5 = Create Widget!(
    execute:Exec=@n4.else
    Class:User Widget Class Reference=/Game/UI/Inventory/Cr_ItemTag.Cr_ItemTag_C
    OwningPlayer:Player Controller Object Reference=GetOwningPlayer()
    TagText:Text=@n3.DisplayText
    IsCreativeBetaTag:bool=False
   ) -> (then:Exec, ReturnValue:Cr Item Tag Object Reference)
   > Update!(execute, self:Cr Item Tag Object Reference=@n5, inTagText:Text=@n3.DisplayText) -> (then)
   > n6 = AddChildToWrapBox!(execute, self:Wrap Box Object Reference=Get WrapBox_Tags(), Content:Widget Object Reference=@n5) -> (then, ReturnValue:Wrap Box Slot Object Reference)
   > SetPadding!(execute, self:Wrap Box Slot Object Reference=@n6, InPadding:Margin Structure=Make Margin(Left=1, Top=0, Right=1, Bottom=0)) -> (then)
  ]}
 SetCodeVisibilityByPrivacy {nodes [
   n1 = SetCodeVisibilityByPrivacy!() -> (then, NewPrivacy:"EMMSPrivacy Enum (by ref)"=Public) {is=title K2Node_FunctionEntry}
   > n2 = Switch on EMMSPrivacy!(execute, Selection:EMMSPrivacy Enum=@n1) -> (Public:Exec, Private:Exec) {is=title K2Node_SwitchEnum}
   n3 = Get Text_LinkCode() -> (Text_LinkCode:Common Text Object Reference)
   SetVisibility(execute=@n2.Private, self=@n3, InVisibility=Collapsed)
   SetVisibility(execute=@n2.Public, self=@n3, InVisibility=HitTestInvisible)
  ]}
 SetCuratedStatus {nodes [
   n1 = SetCuratedStatus!() -> (then, bIsFeatured:bool=~, bIsLinkCode:bool=~, bIsUserInitiatedLoad:bool=~) {is=title K2Node_FunctionEntry}
   > SetVisibility(self=Get Text_FeaturedCreation() -> (Text_FeaturedCreation:Common Text Object Reference), InVisibility=Collapsed)
   > n2 = Set bIsCurated(bIsCurated:bool=@n1) -> (Output_Get:bool)
   > n3 = Set IsLinkCode(IsLinkCode:bool=@n1.bIsLinkCode) -> (Output_Get:bool)
   > n4 = Branch(Condition=AND Boolean(A=@n2, B=NOT Boolean(A=@n1.bIsUserInitiatedLoad)))
   > SetVisibility(self=Get Text_FeaturedCreation(), InVisibility=SelfHitTestInvisible)
   Branch(execute=@n4.else, Condition=NOT Boolean(A=@n3))
  ]}
 SetDescription {nodes [
   n1 = SetDescription!() -> (then, Description:Text) {is=title K2Node_FunctionEntry}
   n2 = TextIsEmpty(InText=@n1)
   > Set HasDescription(HasDescription:bool=NOT Boolean(A=@n2)) -> (Output_Get:bool)
   n3 = Get Text_Description() -> (Text_Description:Common Text Object Reference)
   > "SetText (Text)"(self=@n3, InText=@n1)
   > SetVisibility(self=@n3, InVisibility=Select(Option 0=SelfHitTestInvisible, Option 1=Collapsed, Index=@n2))
   > PopulateTags!(
    execute
    self
    TagArray:Array of Creative Island Description Tag Structures=GetIslandDescriptionTags!(self:Self Object Reference) -> (DescriptionTags:Array of Creative Island Description Tag Structures)
   ) -> (then) {is=title K2Node_CallFunction}
  ]}
 SetIslandName {nodes [
   n1 = SetIslandName!() -> (then, NewIslandName:Text) {is=title K2Node_FunctionEntry}
   > n2 = Set IslandName(IslandName:Text=@n1) -> (Output_Get:Text)
   n3 = Get Text_IslandName() -> (Text_IslandName:Common Text Object Reference)
   > "SetText (Text)"(self=@n3, InText=TextToUpper!(InText:"Text (by ref)"=@n2) -> (ReturnValue:Text))
   > SetVisibility(self=@n3, InVisibility=Select(Option 0=SelfHitTestInvisible, Option 1=Collapsed, Index=TextIsEmpty(InText=@n1)))
  ]}
 SetIslandPop {nodes [
   n1 = SetIslandPop!() -> (then, NewPop:int, MaxPop:int) {is=title K2Node_FunctionEntry}
   > n2 = Compare Int!(Exec:Exec, Input:int=@n1, CompareWith:int) -> (">":Exec, "==":Exec, "<":Exec) {is=title K2Node_MacroInstance}
   n3 = Get Text_IslandPop() -> (Text_IslandPop:Common Text Object Reference)
   "SetText (Text)"(execute=@n2."<", self=@n3)
   n4 = Get SocialImage() -> (SocialImage:Image Object Reference)
   > SetVisibility(self=@n4, InVisibility=Collapsed)
   "SetText (Text)"(execute=[@n2."==", @n2.">"], self=@n3, InText=Format Text!(Format="LOCTEXT(\"{x}/{y}\")", x:int=@n1, y:int=@n1.MaxPop) -> (Result))
   > SetVisibility(self=@n4, InVisibility=HitTestInvisible)
  ]}
 SetLinkCode {nodes [
   n1 = SetLinkCode!() -> (then, LinkCode:String) {is=title K2Node_FunctionEntry}
   n2 = Get Text_LinkCode()
   > "SetText (Text)"(self=@n2, InText="To Text (String)"(InString=@n1))
   n3 = "Equal (Integer)"(A=Len(S=@n1), B=0) {is=title K2Node_CallFunction}
   > SetVisibility(self=@n2, InVisibility=Select(Option 0=SelfHitTestInvisible, Option 1=Collapsed, Index=@n3))
   > n4 = Branch(Condition=@n3)
   SetCodeVisibilityByPrivacy!(
    execute=@n4.else
    self
    NewPrivacy:"EMMSPrivacy Enum (by ref)"=GetPrivacy!(self:Fort Athena Creative Portal Object Reference=Get MyBGAIslandPortal()) -> (ReturnValue:EMMSPrivacy Enum=Public)
   ) -> (then) {is=title K2Node_CallFunction}
  ]}
 SetPlayerName {nodes [
   n1 = SetPlayerName!() -> (then, PlayerName:Text) {is=title K2Node_FunctionEntry}
   > Set PlayerName(PlayerName:Text=@n1) -> (Output_Get:Text)
   > "SetText (Text)"(self=Get Text_PlayerName() -> (Text_PlayerName:Common Text Object Reference), InText=@n1)
   > SetVisibility(self=Get Text_PlayerName(), InVisibility=Select(Option 0=SelfHitTestInvisible, Option 1=Hidden, Index=TextIsEmpty(InText=@n1)))
  ]}
 SetupPortalEventBinding {nodes [
   n1 = SetupPortalEventBinding!() -> (then, InPortal:BGA Island Portal Object Reference, Bind:bool=~) {is=title K2Node_FunctionEntry}
   > n2 = Branch(Condition=@n1.Bind)
   > Bind Event to Update Island Name(self:BGA Island Portal Object Reference=@n1, Delegate=Create Event()) {is=title K2Node_AddDelegate}
   > Bind Event to Update Player Name(self:BGA Island Portal Object Reference=@n1, Delegate=Create Event()) {is=title K2Node_AddDelegate}
   > Bind Event to Update Player Pop(self:BGA Island Portal Object Reference=@n1, Delegate=Create Event()) {is=title K2Node_AddDelegate}
   > Bind Event to Update Is Curated Portal(self:BGA Island Portal Object Reference=@n1, Delegate=Create Event()) {is=title K2Node_AddDelegate}
   > Bind Event to Update Description(self:BGA Island Portal Object Reference=@n1, Delegate=Create Event()) {is=title K2Node_AddDelegate}
   > Bind Event to Update Link Code(self:BGA Island Portal Object Reference=@n1, Delegate=Create Event()) {is=title K2Node_AddDelegate}
   Unbind Event from Update Island Name(execute=@n2.else, self:BGA Island Portal Object Reference=@n1, Delegate=Create Event()) {is=title K2Node_RemoveDelegate}
   > Unbind Event from Update Player Name(self:BGA Island Portal Object Reference=@n1, Delegate=Create Event()) {is=title K2Node_RemoveDelegate}
   > Unbind Event from Update Player Pop(self:BGA Island Portal Object Reference=@n1, Delegate=Create Event()) {is=title K2Node_RemoveDelegate}
   > Unbind Event from Update Is Curated Portal(self:BGA Island Portal Object Reference=@n1, Delegate=Create Event()) {is=title K2Node_RemoveDelegate}
   > Unbind Event from Update Description(self:BGA Island Portal Object Reference=@n1, Delegate=Create Event()) {is=title K2Node_RemoveDelegate}
   > Unbind Event from Update Link Code(self:BGA Island Portal Object Reference=@n1, Delegate=Create Event()) {is=title K2Node_RemoveDelegate}
  ]}
 UpdateVisBasedOnDistance {nodes [
   UpdateVisBasedOnDistance() {is=title K2Node_FunctionEntry}
   n1 = AND Boolean(A=Get HasDescription() -> (HasDescription:bool), B=Get bShouldShowDescription() -> (bShouldShowDescription:bool))
   > Set bNewCurVis(bNewCurVis:bool=@n1) -> (Output_Get:bool)
   > SetVisibility(self=Get Text_Description(), InVisibility=Select(Option 0=Collapsed, Option 1=SelfHitTestInvisible, Index=@n1))
   > n2 = Branch(Condition="Equal (Boolean)"!(A:bool=Get bNewCurVis() -> (bNewCurVis:bool), B:bool=Get IsCurVis() -> (IsCurVis:bool)) -> (ReturnValue:bool))
   n3 = Set IsCurVis(execute=@n2.else, IsCurVis:bool=@n1) -> (Output_Get:bool)
   > OnNewVisibilityThresholdChanged!(execute, self, bNewVis:bool=@n3) -> (then) {is=title K2Node_CallFunction}
  ]}
]
variables [
 bIsCurated {type=boolean}
 bNewCurVis {type=boolean}
 bShouldShowDescription {type=boolean}
 HasDescription {type=boolean}
 IsCurVis {type=boolean}
 IslandName {type=string, default=""}
 IsLinkCode {type=boolean}
 MyBGAIslandPortal {type=BGA_IslandPortal_C, default=null}
 PlayerName {type=string, default=""}
]
widgets {
 info {inheritedWidgetCount=0, parentClass=&UserWidget, rootWidgetClass=&Overlay, widgetCount=17}
 widgets [
  Overlay_2 {widgetClassPath=&Overlay, widget=Overlay_2}
  VerticalBox_0 {
   widgetClassPath=&VerticalBox
   parent=Overlay_2
   widget=VerticalBox_0
   slot {class=OverlaySlot, properties {padding=<ltrb 0 6 0 0>}}
  }
  Text_FeaturedCreation {
   widgetClassPath=&CommonTextBlock
   bIsVariable=true
   parent=VerticalBox_0
   widget=Text_FeaturedCreation
   properties {
    justification=Center
    text=FEATURED CREATION
    colorAndOpacity {colorUseRule=UseColor_Specified, specifiedColor=<rgba 0.9301 1 0.0091 1>}
    font {
     bForceMonospaced=false
     bMaterialIsStencil=false
     fontMaterial=null
     fontObject=&BurbankSmall
     letterSpacing=0
     monospacedWidth=1
     size=12
     skewAmount=0
     typefaceFontName=Black
     outlineSettings {
      bApplyOutlineToDropShadows=false
      bMiteredCorners=false
      bSeparateFillAlpha=false
      outlineMaterial=null
      outlineSize=1
      outlineColor=<rgba 0.0052 0.0052 0.0056 0.25>
     }
    }
    renderTransform {angle=0, scale=<1 1>, shear=<-8 0>, translation=<0 3>}
    shadowColorAndOpacity=<rgba 0.0052 0.0052 0.0056 0>
   }
   slot {class=VerticalBoxSlot, properties {horizontalAlignment=HAlign_Left, padding=<ltrb 12 0 0 0>}}
  }
  IslandNameBox {
   widgetClassPath=&VerticalBox
   bIsVariable=true
   parent=VerticalBox_0
   widget=IslandNameBox
   slot {class=VerticalBoxSlot, properties {horizontalAlignment=HAlign_Left}}
  }
  Text_IslandName {
   widgetClassPath=&CommonTextBlock
   bIsVariable=true
   parent=IslandNameBox
   widget=Text_IslandName
   properties {
    text="Tiny Temple "
    textTransformPolicy=ToUpper
    wrapTextAt=465
    font {
     bForceMonospaced=false
     bMaterialIsStencil=false
     fontMaterial=null
     fontObject=&BurbankSmall
     letterSpacing=0
     monospacedWidth=1
     size=24
     skewAmount=0
     typefaceFontName=Black
     outlineSettings {
      bApplyOutlineToDropShadows=false
      bMiteredCorners=false
      bSeparateFillAlpha=false
      outlineMaterial=null
      outlineSize=1
      outlineColor=<rgba 0.0052 0.0052 0.0056 0.2>
     }
    }
    renderTransform {angle=0, scale=<1 1>, shear=<-8 0>, translation=<0 0>}
    shadowColorAndOpacity=<rgba 0.0052 0.0052 0.0056 0.822>
    shadowOffset=<0 0>
   }
   slot {class=VerticalBoxSlot, properties {horizontalAlignment=HAlign_Left, padding=<ltrb 12 0 0 0>}}
  }
  VerticalBox_1 {
   widgetClassPath=&VerticalBox
   parent=VerticalBox_0
   widget=VerticalBox_1
   slot {class=VerticalBoxSlot, properties {horizontalAlignment=HAlign_Left, padding=<ltrb 12 0 0 0>}}
  }
  Text_LinkCode {
   widgetClassPath=&CommonTextBlock
   bIsVariable=true
   parent=VerticalBox_1
   widget=Text_LinkCode
   properties {
    justification=Center
    text=4536-7842-2454
    font {
     bForceMonospaced=false
     bMaterialIsStencil=false
     fontMaterial=null
     fontObject=&BurbankSmall
     letterSpacing=0
     monospacedWidth=1
     size=18
     skewAmount=0
     typefaceFontName=Black
     outlineSettings {
      bApplyOutlineToDropShadows=false
      bMiteredCorners=false
      bSeparateFillAlpha=false
      outlineMaterial=null
      outlineSize=1
      outlineColor=<rgba 0.0052 0.0052 0.0056 0.2>
     }
    }
    renderTransform {angle=0, scale=<1 1>, shear=<-8 0>, translation=<0 0>}
    shadowColorAndOpacity=<rgba 0.0052 0.0052 0.0056 0.822>
    shadowOffset=<0 0>
   }
   slot {class=VerticalBoxSlot, properties {horizontalAlignment=HAlign_Left}}
  }
  HorizontalBox_0 {widgetClassPath=&HorizontalBox, parent=VerticalBox_1, widget=HorizontalBox_0, slot {class=VerticalBoxSlot}}
  CommonTextBlock_75 {
   widgetClassPath=&CommonTextBlock
   parent=HorizontalBox_0
   widget=CommonTextBlock_75
   properties {
    text="Created by: "
    textTransformPolicy=ToUpper
    colorAndOpacity {colorUseRule=UseColor_Specified, specifiedColor=<rgba 1 1 1 0.8>}
    font {
     bForceMonospaced=false
     bMaterialIsStencil=false
     fontMaterial=null
     fontObject=&BurbankSmall
     letterSpacing=0
     monospacedWidth=1
     size=14
     skewAmount=0
     typefaceFontName=Black
     outlineSettings {
      bApplyOutlineToDropShadows=false
      bMiteredCorners=false
      bSeparateFillAlpha=false
      outlineMaterial=null
      outlineSize=1
      outlineColor=<rgba 0.0052 0.0052 0.0056 0.149>
     }
    }
    renderTransform {angle=0, scale=<1 1>, shear=<-8 0>, translation=<0 0>}
    shadowOffset=<0 0>
   }
   slot {class=HorizontalBoxSlot, properties {verticalAlignment=VAlign_Bottom}}
  }
  Text_PlayerName {
   widgetClassPath=&CommonTextBlock
   bIsVariable=true
   parent=HorizontalBox_0
   widget=Text_PlayerName
   properties {
    justification=Center
    text=LONGNAMEPERSON
    visibility=Hidden
    colorAndOpacity {colorUseRule=UseColor_Specified, specifiedColor=<rgba 0 1 1 1>}
    font {
     bForceMonospaced=false
     bMaterialIsStencil=false
     fontMaterial=null
     fontObject=&BurbankSmall
     letterSpacing=0
     monospacedWidth=1
     size=16
     skewAmount=0
     typefaceFontName=Black
     outlineSettings {
      bApplyOutlineToDropShadows=false
      bMiteredCorners=false
      bSeparateFillAlpha=false
      outlineMaterial=null
      outlineSize=1
      outlineColor=<rgba 0 0.066 0.066 0.3>
     }
    }
    renderTransform {angle=0, scale=<1 1>, shear=<-8 0>, translation=<0 0>}
    shadowColorAndOpacity=<rgba 0.0052 0.0052 0.0056 0.822>
    shadowOffset=<0 0>
   }
   slot {class=HorizontalBoxSlot, properties {verticalAlignment=VAlign_Bottom, padding=<ltrb 4 0 0 0>}}
  }
  PopulationBox {
   widgetClassPath=&HorizontalBox
   bIsVariable=true
   parent=HorizontalBox_0
   widget=PopulationBox
   slot {class=HorizontalBoxSlot, properties {horizontalAlignment=HAlign_Center, padding=<ltrb 20 0 10 0>}}
  }
  SocialImage {
   widgetClassPath=&Image
   bIsVariable=true
   parent=PopulationBox
   widget=SocialImage
   properties {brush {
     bIsDynamicallyLoaded=false
     drawAs=Image
     imageType=NoImage
     mirroring=NoMirror
     resourceName=None
     resourceObject=/Game/UI/Frontend/Parties/PartiesButtonStyles/T_UI_Icon_QuadMeeple
     tiling=NoTile
     imageSize=<20 20>
     margin=<ltrb 0 0 0 0>
     outlineSettings {
      bUseBrushTransparency=false
      roundingType=HalfHeightRadius
      width=0
      color {colorUseRule=UseColor_Specified, specifiedColor=<rgba 0 0 0 0>}
      cornerRadii=<0 0 0 0>
     }
     tintColor {colorUseRule=UseColor_Specified, specifiedColor=<rgba 1 1 1 1>}
     uVRegion {bIsValid=false, max=<0 0>, min=<0 0>}
    }}
   slot {class=HorizontalBoxSlot, properties {verticalAlignment=VAlign_Center}}
  }
  Text_IslandPop {
   widgetClassPath=&CommonTextBlock
   bIsVariable=true
   parent=PopulationBox
   widget=Text_IslandPop
   properties {
    text=00/00
    font {
     bForceMonospaced=false
     bMaterialIsStencil=false
     fontMaterial=null
     fontObject=&BurbankSmall
     letterSpacing=0
     monospacedWidth=1
     size=16
     skewAmount=0
     typefaceFontName=Black
     outlineSettings {
      bApplyOutlineToDropShadows=false
      bMiteredCorners=false
      bSeparateFillAlpha=false
      outlineMaterial=null
      outlineSize=1
      outlineColor=<rgba 0.0052 0.0052 0.0056 0.149>
     }
    }
    shadowColorAndOpacity=<rgba 0.0052 0.0052 0.0056 0.822>
    shadowOffset=<0 0>
   }
   slot {class=HorizontalBoxSlot, properties {verticalAlignment=VAlign_Bottom, padding=<ltrb 0 0 0 -2>}}
  }
  WrapBox_Tags {
   widgetClassPath=&WrapBox
   bIsVariable=true
   parent=VerticalBox_0
   widget=WrapBox_Tags
   slot {class=VerticalBoxSlot, properties {padding=<ltrb 12 6 12 6>}}
  }
  Text_Description {
   widgetClassPath=&CommonTextBlock
   bIsVariable=true
   parent=VerticalBox_0
   widget=Text_Description
   properties {
    text=Suspiciously long vague description text about this that or the other WHO KNOWS
    wrapTextAt=465
    colorAndOpacity {colorUseRule=UseColor_Specified, specifiedColor=<rgba 0 1 1 1>}
    font {
     bForceMonospaced=false
     bMaterialIsStencil=false
     fontMaterial=null
     fontObject=&BurbankSmall
     letterSpacing=0
     monospacedWidth=1
     size=12
     skewAmount=0
     typefaceFontName=Bold
     outlineSettings {
      bApplyOutlineToDropShadows=false
      bMiteredCorners=false
      bSeparateFillAlpha=false
      outlineMaterial=null
      outlineSize=1
      outlineColor=<rgba 0.0052 0.0052 0.0056 0.149>
     }
    }
    shadowColorAndOpacity=<rgba 0.0052 0.0052 0.0056 0.822>
    shadowOffset=<0 0>
   }
   slot {class=VerticalBoxSlot, properties {horizontalAlignment=HAlign_Left, padding=<ltrb 12 0 0 0>}}
  }
  Border_109 {
   widgetClassPath=&Border
   parent=VerticalBox_0
   widget=Border_109
   properties {
    background {
     bIsDynamicallyLoaded=false
     drawAs=Image
     imageType=NoImage
     mirroring=NoMirror
     resourceName=None
     resourceObject=null
     tiling=NoTile
     imageSize=<32 32>
     margin=<ltrb 0 0 0 0>
     outlineSettings {
      bUseBrushTransparency=false
      roundingType=HalfHeightRadius
      width=0
      color {colorUseRule=UseColor_Specified, specifiedColor=<rgba 0 0 0 0>}
      cornerRadii=<0 0 0 0>
     }
     tintColor {colorUseRule=UseColor_Specified, specifiedColor=<rgba 0 1 1 1>}
     uVRegion {bIsValid=false, max=<0 0>, min=<0 0>}
    }
    padding=<ltrb 4 4 4 2>
   }
   slot {class=VerticalBoxSlot, properties {padding=<ltrb 0 6 0 0>}}
  }
  CreativeLogoWatermark {
   widgetClassPath=&Image
   bIsVariable=true
   parent=Overlay_2
   widget=CreativeLogoWatermark
   properties {
    visibility=Collapsed
    brush {
     bIsDynamicallyLoaded=false
     drawAs=Image
     imageType=NoImage
     mirroring=NoMirror
     resourceName=None
     resourceObject=/Game/Athena/HUD/Art/Materials/M_UI_CreativeLogo
     tiling=NoTile
     imageSize=<80 80>
     margin=<ltrb 0 0 0 0>
     outlineSettings {
      bUseBrushTransparency=false
      roundingType=HalfHeightRadius
      width=0
      color {colorUseRule=UseColor_Specified, specifiedColor=<rgba 0 0 0 0>}
      cornerRadii=<0 0 0 0>
     }
     tintColor {colorUseRule=UseColor_Specified, specifiedColor=<rgba 1 1 1 1>}
     uVRegion {bIsValid=false, max=<0 0>, min=<0 0>}
    }
    renderTransform {angle=-15, scale=<1 1>, shear=<0 0>, translation=<-26 -24>}
   }
   slot {class=OverlaySlot}
  }
 ]
}
