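# HUD interaction detail widget for Creative island portals, showing a "Change Destination" title and a status line.
# InitFromObject casts to BGA_IslandPortal, stores it and binds to its curated-status and player-population delegates.
# UpdateInteractionToolTip sets the line to an interact prompt, "Match in Progress" or "Portal Destination Locked", based on featured status, ownership, interactability and population.
# PlayIntro runs a short fade, scale and rotate animation on the logo watermark, which defaults to collapsed.
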WidgetBlueprint /Game/Playgrounds/Items/Portals/UI/CreativePortalCodeScreenHUDDetails
animations [Intro {
  binding_count=1
  length_seconds=0.25
  object=&self:Intro
  sequence {bindings [CreativeLogoWatermark {
     id=#1
     possessed_class=Image
     tracks [
      {
       type=MovieSceneFloatTrack
       display_name=Render Opacity
       sections [{type=MovieSceneFloatSection, end_frame=5, start_frame=0, channels [None {type=MovieSceneScriptingFloatChannel, keys [[0,0],[5,1]]}]}]
      }
      {
       type=MovieScene2DTransformTrack
       sections [{
         type=MovieScene2DTransformSection
         end_frame=5
         start_frame=0
         channels [
          Translation.X {type=MovieSceneScriptingFloatChannel, const [-26,1]}
          Translation.Y {type=MovieSceneScriptingFloatChannel, const [-24,1]}
          Angle {type=MovieSceneScriptingFloatChannel, keys [[0,8],[5,-15]]}
          Scale.X {type=MovieSceneScriptingFloatChannel, keys [[0,2],[5,1]]}
          Scale.Y {type=MovieSceneScriptingFloatChannel, keys [[0,2],[5,1]]}
          Shear.X {type=MovieSceneScriptingFloatChannel, const [0,1]}
          Shear.Y {type=MovieSceneScriptingFloatChannel, const [0,1]}
         ]
        }]
      }
     ]
    }]}
 }]
graphs [
 EventGraph {nodes [
   n1 = Event InitFromObject!() -> (OutputDelegate, then, InitObject:Object Reference) {is=title K2Node_Event}
   > n2 = Cast To BGA_IslandPortal!(execute, Object=@n1.InitObject) -> (then, CastFailed, AsBGA Island Portal:BGA Island Portal Object Reference) {is=title K2Node_DynamicCast}
   > n3 = Set MyBGAIslandPortal(MyBGAIslandPortal:BGA Island Portal Object Reference=@n2) -> (Output_Get:BGA Island Portal Object Reference)
   > SetupPortalEventBinding!(execute, self, InPortal:BGA Island Portal Object Reference=@n3, bBind:bool=true) -> (then)
   Event PlayIntro() {is=title K2Node_Event}
   > PlayAnimationForward(InAnimation=Get Intro())
  ]}
 OnPortalPopulationUpdated {nodes [
   OnPortalPopulationUpdated!() -> (then, NewPopulation:int, MaxPopulation:int) {is=title K2Node_FunctionEntry}
   > UpdateInteractionToolTip!(
    execute
    self
    IsFeatured:bool=GetIsPublishedPortal!(self:Fort Athena Creative Portal Object Reference=Get MyBGAIslandPortal() -> (MyBGAIslandPortal:BGA Island Portal Object Reference)) -> (ReturnValue:bool)
   ) -> (then)
  ]}
 SetCuratedStatus {nodes [
   n1 = SetCuratedStatus!() -> (then, bIsFeatured:bool=~, bIsLinkCode:bool=~, bIsUserInitiatedLoad:bool=~) {is=title K2Node_FunctionEntry}
   > UpdateInteractionToolTip(IsFeatured=@n1)
  ]}
 SetupPortalEventBinding {nodes [
   n1 = SetupPortalEventBinding!() -> (then, InPortal:BGA Island Portal Object Reference, bBind:bool=~) {is=title K2Node_FunctionEntry}
   > n2 = Branch(Condition=@n1.bBind)
   n3 = Create Event()
   > Bind Event to Update Is Curated Portal(self:BGA Island Portal Object Reference=@n1, Delegate=@n3) {is=title K2Node_AddDelegate}
   n4 = Create Event()
   > Bind Event to Update Player Pop(self:BGA Island Portal Object Reference=@n1, Delegate=@n4) {is=title K2Node_AddDelegate}
   Unbind Event from Update Is Curated Portal(execute=@n2.else, self:BGA Island Portal Object Reference=@n1, Delegate=@n3) {is=title K2Node_RemoveDelegate}
   > Unbind Event from Update Player Pop(self:BGA Island Portal Object Reference=@n1, Delegate=@n4) {is=title K2Node_RemoveDelegate}
  ]}
 UpdateInteractionToolTip {nodes [
   UpdateInteractionToolTip!() -> (then, IsFeatured:bool=~) {is=title K2Node_FunctionEntry}
   > Branch(Condition=IsValid(Object=Get MyBGAIslandPortal()))
   n1 = Get MyBGAIslandPortal()
   > n2 = Branch(Condition=OR Boolean(A=Get IsFeatured() -> (IsFeatured:bool), B=Get IsMyPortal!(self:BGA Island Portal Object Reference=@n1) -> (IsMyPortal:bool)))
   > n3 = Branch(Condition=CanUseHUDInteract!(self:BGA Island Portal Object Reference=Get MyBGAIslandPortal()) -> (CanInteract:bool))
   > "SetText (Text)"(
    self=Get Text_InteractionDescription() -> (Text_InteractionDescription:Common Text Object Reference)
    InText=Select(Option 0:Text="LOCTEXT(\"Interact to set your island\")", Option 1:Text="LOCTEXT(\"Interact to set code\")", Index=Get IsFeatured()) -> (ReturnValue:Text)
   )
   n4 = "Equal (Byte)"!(
    A:Byte=Get CurrentPopulation!(self:Fort Athena Creative Portal Object Reference=Get MyBGAIslandPortal()) -> (CurrentPopulation:Byte=0)
    B:Byte=0
   ) -> (ReturnValue:bool)
   n5 = Get IsMyPortal(self=Get MyBGAIslandPortal())
   "SetText (Text)"(
    execute=@n3.else
    self=Get Text_InteractionDescription()
    InText=Select(Option 0="LOCTEXT(\"Match in Progress\")", Option 1="LOCTEXT(\"Portal Destination Locked\")", Index=OR Boolean(A=@n4, B=@n5))
   )
   "SetText (Text)"(execute=@n2.else, self=Get Text_InteractionDescription(), InText="LOCTEXT(\"Portal Destination Locked\")")
   Branch(Condition=IsValid(Object=Get MyBGAIslandPortal()))
   n6 = Get MyBGAIslandPortal()
   > n7 = Branch(Condition=OR Boolean(A=Get IsFeatured(), B=Get IsMyPortal(self=@n6)))
   n8 = Get MyBGAIslandPortal()
   n9 = Get CurrentPopulation(self=@n8) -> (CurrentPopulation=0)
   n10 = CanUseHUDInteract(self=@n8)
   n11 = Select(Option 0="LOCTEXT(\"Interact to set your island\")", Option 1="LOCTEXT(\"Interact to set code\")", Index=false)
   > "SetText (Text)"(
    self=Get Text_InteractionDescription()
    InText=Select(
     Option 0=Select(Option 0="LOCTEXT(\"Match in Progress\")", Option 1=@n11, Index=@n10)
     Option 1="LOCTEXT(\"Portal Destination Locked\")"
     Index="Equal (Byte)"(A=@n9, B=0)
    )
   )
   "SetText (Text)"(execute=@n7.else, self=Get Text_InteractionDescription(), InText="LOCTEXT(\"Portal Destination Locked\")")
  ]}
]
variables [MyBGAIslandPortal {type=BGA_IslandPortal_C, default=null}]
widgets {
 info {inheritedWidgetCount=0, parentClass=&UserWidget, rootWidgetClass=&Overlay, widgetCount=7}
 widgets [
  Overlay_28 {widgetClassPath=&Overlay, widget=Overlay_28}
  VerticalBox_56 {
   widgetClassPath=&VerticalBox
   parent=Overlay_28
   widget=VerticalBox_56
   slot {class=OverlaySlot, properties {padding=<ltrb 0 6 0 0>}}
  }
  InteractionNameBox {
   widgetClassPath=&VerticalBox
   bIsVariable=true
   parent=VerticalBox_56
   widget=InteractionNameBox
   slot {class=VerticalBoxSlot, properties {horizontalAlignment=HAlign_Left}}
  }
  Text_InteractionName {
   widgetClassPath=&CommonTextBlock
   bIsVariable=true
   parent=InteractionNameBox
   widget=Text_InteractionName
   properties {
    text=CHANGE DESTINATION
    textTransformPolicy=ToUpper
    wrapTextAt=465
    font {
     bForceMonospaced=false
     bMaterialIsStencil=false
     fontMaterial=null
     fontObject=&BurbankSmall
     letterSpacing=0
     monospacedWidth=1
     size=24
     skewAmount=0
     typefaceFontName=Black
     outlineSettings {
      bApplyOutlineToDropShadows=false
      bMiteredCorners=false
      bSeparateFillAlpha=false
      outlineMaterial=null
      outlineSize=1
      outlineColor=<rgba 0.0052 0.0052 0.0056 0.2>
     }
    }
    renderTransform {angle=0, scale=<1 1>, shear=<-8 0>, translation=<0 0>}
    shadowColorAndOpacity=<rgba 0.0052 0.0052 0.0056 0.822>
    shadowOffset=<0 0>
   }
   slot {class=VerticalBoxSlot, properties {horizontalAlignment=HAlign_Left, padding=<ltrb 12 0 0 0>}}
  }
  Text_InteractionDescription {
   widgetClassPath=&CommonTextBlock
   bIsVariable=true
   parent=VerticalBox_56
   widget=Text_InteractionDescription
   properties {
    justification=Center
    text=Interact to set Code
    colorAndOpacity {colorUseRule=UseColor_Specified, specifiedColor=<rgba 0.9301 1 0.0091 1>}
    font {
     bForceMonospaced=false
     bMaterialIsStencil=false
     fontMaterial=null
     fontObject=&BurbankSmall
     letterSpacing=0
     monospacedWidth=1
     size=12
     skewAmount=0
     typefaceFontName=Black
     outlineSettings {
      bApplyOutlineToDropShadows=false
      bMiteredCorners=false
      bSeparateFillAlpha=false
      outlineMaterial=null
      outlineSize=1
      outlineColor=<rgba 0.0052 0.0052 0.0056 0.25>
     }
    }
    renderTransform {angle=0, scale=<1 1>, shear=<-8 0>, translation=<0 3>}
    shadowColorAndOpacity=<rgba 0.0052 0.0052 0.0056 0>
   }
   slot {class=VerticalBoxSlot, properties {horizontalAlignment=HAlign_Left, padding=<ltrb 12 0 0 0>}}
  }
  Border_109 {
   widgetClassPath=&Border
   parent=VerticalBox_56
   widget=Border_109
   properties {
    background {
     bIsDynamicallyLoaded=false
     drawAs=Image
     imageType=NoImage
     mirroring=NoMirror
     resourceName=None
     resourceObject=null
     tiling=NoTile
     imageSize=<32 32>
     margin=<ltrb 0 0 0 0>
     outlineSettings {
      bUseBrushTransparency=false
      roundingType=HalfHeightRadius
      width=0
      color {colorUseRule=UseColor_Specified, specifiedColor=<rgba 0 0 0 0>}
      cornerRadii=<0 0 0 0>
     }
     tintColor {colorUseRule=UseColor_Specified, specifiedColor=<rgba 0 1 1 1>}
     uVRegion {bIsValid=false, max=<0 0>, min=<0 0>}
    }
    padding=<ltrb 4 4 4 2>
   }
   slot {class=VerticalBoxSlot, properties {padding=<ltrb 0 6 0 0>}}
  }
  CreativeLogoWatermark {
   widgetClassPath=&Image
   bIsVariable=true
   parent=Overlay_28
   widget=CreativeLogoWatermark
   properties {
    visibility=Collapsed
    brush {
     bIsDynamicallyLoaded=false
     drawAs=Image
     imageType=NoImage
     mirroring=NoMirror
     resourceName=None
     resourceObject=/Game/Athena/HUD/Art/Materials/M_UI_CreativeLogo
     tiling=NoTile
     imageSize=<80 80>
     margin=<ltrb 0 0 0 0>
     outlineSettings {
      bUseBrushTransparency=false
      roundingType=HalfHeightRadius
      width=0
      color {colorUseRule=UseColor_Specified, specifiedColor=<rgba 0 0 0 0>}
      cornerRadii=<0 0 0 0>
     }
     tintColor {colorUseRule=UseColor_Specified, specifiedColor=<rgba 1 1 1 1>}
     uVRegion {bIsValid=false, max=<0 0>, min=<0 0>}
    }
    renderTransform {angle=-15, scale=<1 1>, shear=<0 0>, translation=<-26 -24>}
   }
   slot {class=OverlaySlot}
  }
 ]
}
