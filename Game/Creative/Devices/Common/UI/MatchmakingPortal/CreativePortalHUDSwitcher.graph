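# Widget Blueprint for the Creative island portal HUD that switches between a portal details view and a code screen details view.
# InitFromObject forwards the object to both child views, stores it as a BGA_IslandPortal and selects the view from whether the local player is looking at the screen.
# It also binds to the portal's looking-at-screen update event.
# PlayIntro is relayed to the active view through the details interface.
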WidgetBlueprint /Game/Creative/Devices/Common/UI/MatchmakingPortal/CreativePortalHUDSwitcher
graphs [
 EventGraph {nodes [
   Event PlayIntro() {is=title K2Node_Event}
   > n1 = Cast To CreativePortalHUDDetailsInterface!(
    execute
    Object=GetActiveWidget!(self:Widget Switcher Object Reference=Get PortalHUDSwitcher() -> (PortalHUDSwitcher:Common Widget Switcher Legacy Object Reference)) -> (ReturnValue:Widget Object Reference)
   ) -> (then, CastFailed, AsCreative Portal HUDDetails Interface:Creative Portal HUDDetails Interface Interface) {is=title K2Node_DynamicCast}
   > PlayIntro(self:Creative Portal HUDDetails Interface Interface=@n1)
   n2 = Event InitFromObject!() -> (OutputDelegate, then, InitObject:Object Reference) {is=title K2Node_Event}
   > InitFromObject!(
    execute
    self:Fort Initialize From Object Interface Interface=Get CreativePortalHUDDetails() -> (CreativePortalHUDDetails:Creative Portal HUDDetails Object Reference)
    InitObject:Object Reference=@n2.InitObject
   ) -> (then)
   > InitFromObject(
    self=Get CreativePortalCodeScreenHUDDetails() -> (CreativePortalCodeScreenHUDDetails:Creative Portal Code Screen HUDDetails Object Reference)
    InitObject=@n2.InitObject
   )
   > n3 = Cast To BGA_IslandPortal!(execute, Object=@n2.InitObject) -> (then, CastFailed, AsBGA Island Portal:BGA Island Portal Object Reference) {is=title K2Node_DynamicCast}
   > n4 = Set MyBGAIslandPortal(MyBGAIslandPortal:BGA Island Portal Object Reference=@n3) -> (Output_Get:BGA Island Portal Object Reference)
   > SetActivePortalDisplayWidget!(
    execute
    self
    bIsLookingAtCodeScreen:bool=Get bLocalPlayerLookingAtScreen!(self:BGA Island Portal Object Reference=@n4) -> (bLocalPlayerLookingAtScreen:bool)
   ) -> (then)
   > SetupPortalEventBinding!(execute, self, InPortal:BGA Island Portal Object Reference=@n4, bBind:bool=true) -> (then)
  ]}
 SetActivePortalDisplayWidget {nodes [
   n1 = SetActivePortalDisplayWidget!() -> (then, bIsLookingAtCodeScreen:bool=~) {is=title K2Node_FunctionEntry}
   > SetActiveWidget(
    self:Widget Switcher Object Reference=Get PortalHUDSwitcher()
    Widget=Select(
     Option 0:Widget Object Reference=Get CreativePortalHUDDetails()
     Option 1:Widget Object Reference=Get CreativePortalCodeScreenHUDDetails()
     Index=@n1
    ) -> (ReturnValue:Widget Object Reference)
   )
  ]}
 SetupPortalEventBinding {nodes [
   n1 = SetupPortalEventBinding!() -> (then, InPortal:BGA Island Portal Object Reference, bBind:bool=~) {is=title K2Node_FunctionEntry}
   > n2 = Branch(Condition=@n1.bBind)
   > Bind Event to Update Local Player Looking at Screen(self:BGA Island Portal Object Reference=@n1, Delegate=Create Event()) {is=title K2Node_AddDelegate}
   Unbind Event from Update Local Player Looking at Screen(execute=@n2.else, self:BGA Island Portal Object Reference=@n1, Delegate=Create Event()) {is=title K2Node_RemoveDelegate}
  ]}
]
variables [MyBGAIslandPortal {type=BGA_IslandPortal_C, default=null}]
widgets {
 info {inheritedWidgetCount=0, parentClass=&UserWidget, rootWidgetClass=&Overlay, widgetCount=4}
 widgets [
  Overlay_29 {widgetClassPath=&Overlay, widget=Overlay_29}
  PortalHUDSwitcher {
   widgetClassPath=&CommonWidgetSwitcherLegacy
   bIsVariable=true
   parent=Overlay_29
   widget=PortalHUDSwitcher
   slot {class=OverlaySlot}
  }
  CreativePortalHUDDetails {
   widgetClassPath=&/CreativePortalHUDDetails
   bIsVariable=true
   parent=PortalHUDSwitcher
   widget=CreativePortalHUDDetails
   slot {class=WidgetSwitcherSlot}
  }
  CreativePortalCodeScreenHUDDetails {
   widgetClassPath=/Game/Playgrounds/Items/Portals/UI/CreativePortalCodeScreenHUDDetails
   bIsVariable=true
   parent=PortalHUDSwitcher
   widget=CreativePortalCodeScreenHUDDetails
   slot {class=WidgetSwitcherSlot}
  }
 ]
}
