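# Creative portal HUD callout that shows details and an interact prompt for the active interaction context.
# Interaction changes embed the context's optional HUD widget, play its intro through an interface, and set up text, keybind and hold countdown.
# The receiving portal actor is told through an interface when interaction starts or ends.
# A ReadyToShow flag limits backplate fade-in replays, and HUD and keybind events bind on Construct, unbind on Destruct.
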WidgetBlueprint /Game/Creative/Devices/Common/UI/MatchmakingPortal/CreativePortalInfoHUDDisplay
animations [
 Intro {
  binding_count=1
  length_seconds=5
  object=&self:Intro
  sequence {bindings [MarkerPointer {id=#1, possessed_class=AthenaMarkerPointer, tracks []}]}
 }
 FadeInBackplate {
  binding_count=2
  length_seconds=3.75
  object=&self:FadeInBackplate
  sequence {bindings [
    Image_Glow_1 {
     id=#2
     possessed_class=Image
     tracks [{
       type=MovieSceneFloatTrack
       display_name=Render Opacity
       sections [{type=MovieSceneFloatSection, end_frame=40, start_frame=0, channels [None {type=MovieSceneScriptingFloatChannel, keys [[0,1],[30,1],[40,0]]}]}]
      }]
    }
    Image_Backplate {
     id=#3
     possessed_class=Image
     tracks [{
       type=MovieSceneFloatTrack
       display_name=Render Opacity
       sections [{type=MovieSceneFloatSection, end_frame=75, start_frame=0, channels [None {type=MovieSceneScriptingFloatChannel, keys [[0,0],[30,0],[40,1]]}]}]
      }]
    }
   ]}
 }
]
graphs [
 EventGraph {nodes [
   n1 = OnInteractChanged!() -> (OutputDelegate, then, ContextInfo:Fort Interact Context Info Object Reference) {is=title K2Node_CustomEvent}
   > SetActiveContextInfo!(execute, self:Fort Portal HUDDetails Object Reference, ActiveContextInfo:Fort Interact Context Info Object Reference=@n1.ContextInfo) -> (then)
   n2 = GetActiveContextInfo!() -> (ReturnValue:Fort Interact Context Info Object Reference)
   n3 = IsValid(Object=@n2)
   > Set ValidInteraction(ValidInteraction:bool=@n3) -> (Output_Get:bool)
   > n4 = Branch(Condition=@n3)
   > n5 = Branch(Condition=IsAnimationPlaying(InAnimation=Get FadeInBackplate() -> (FadeInBackplate:Widget Animation Object Reference)))
   n6 = IsValid(Object=Get OptionalHUDDisplayWidget!(self:Fort Interact Context Info Object Reference=@n2) -> (OptionalHUDDisplayWidget:User Widget Object Reference))
   n7 = Set ValidInteraction(execute=[@n19.then, @n18.else, @n5.then], ValidInteraction=@n6)
   > SetVisibility(InVisibility=Select(Option 0=Collapsed, Option 1=SelfHitTestInvisible, Index=@n7))
   > n8 = Branch(Condition=@n6)
   n9 = Get VertBox_Content() -> (VertBox_Content:Vertical Box Object Reference)
   > ClearChildren(self=@n9)
   n10 = GetActiveContextInfo()
   n11 = Get OptionalHUDDisplayWidget(self=@n10)
   > AddChildToVerticalBox(self=@n9, Content=@n11)
   > UpdateWidgetVisibility()
   > SetupInteraction()
   > n12 = Sequence()
   > n13 = Set CurReceivedActor(CurReceivedActor:Actor Object Reference=Get ReceivingActor!(self:Fort Interact Context Info Object Reference=@n10) -> (ReceivingActor:Actor Weak object)) -> (Output_Get:Actor Object Reference)
   > n14 = Cast To IslandPortalInteractionInterface!(execute, Object=@n13) -> (then, CastFailed, AsIsland Portal Interaction Interface:Island Portal Interaction Interface Interface) {is=title K2Node_DynamicCast}
   > BroadcastIsBeingInteractedWith!(execute, self:Island Portal Interaction Interface Interface=@n14, IsInteracting:bool=true) -> (then)
   n15 = Cast To CreativePortalHUDDetailsInterface!(execute=@n12.then_1, Object=@n11) -> (then, CastFailed, AsCreative Portal HUDDetails Interface:Creative Portal HUDDetails Interface Interface) {is=title K2Node_DynamicCast}
   > PlayIntro(self:Creative Portal HUDDetails Interface Interface=@n15)
   n16 = Set CurReceivedActor(execute=[@n20.then, @n8.else], CurReceivedActor=Get CurReceivedActor() -> (CurReceivedActor:Actor Object Reference))
   > n17 = Cast To IslandPortalInteractionInterface(Object=@n16)
   > BroadcastIsBeingInteractedWith(self=@n17, IsInteracting=false)
   > Set ReadyToShow(ReadyToShow:bool=true) -> (Output_Get:bool)
   n18 = Branch(execute=@n5.else, Condition=Get ReadyToShow() -> (ReadyToShow:bool))
   > Set ReadyToShow(ReadyToShow=false)
   > n19 = PlayAnimation(InAnimation=Get FadeInBackplate())
   SetVisibility(execute=@n4.else, InVisibility=Collapsed)
   > n20 = StopAnimation(InAnimation=Get FadeInBackplate())
   Event Destruct()
   n21 = Create Event()
   n22 = Get FortHUDContext!() -> (ReturnValue:Fort HUDContext Object Reference) {is=title K2Node_GetBlueprintContextEx}
   > Unbind Event from On Player Can Interact(self:Fort HUDContext Object Reference=@n22, Delegate=@n21) {is=title K2Node_RemoveDelegate}
   n23 = Create Event()
   > Unbind Event from On Interact Updated(self:Fort HUDContext Object Reference=@n22, Delegate=@n23) {is=title K2Node_RemoveDelegate}
   n24 = Create Event()
   n25 = Get FortGlobalUIContext!() -> (ReturnValue:Fort Global UIContext Object Reference) {is=title K2Node_GetBlueprintContextEx}
   > Unbind Event from On Keybinds Changed(self:Fort Global UIContext Object Reference=@n25, Delegate=@n24) {is=title K2Node_RemoveDelegate}
   > ClearChildren(self=Get VertBox_Content())
   Event Construct()
   > Bind Event to On Player Can Interact(self:Fort HUDContext Object Reference=@n22, Delegate=@n21) {is=title K2Node_AddDelegate}
   > Bind Event to On Interact Updated(self:Fort HUDContext Object Reference=@n22, Delegate=@n23) {is=title K2Node_AddDelegate}
   > Bind Event to On Keybinds Changed(self:Fort Global UIContext Object Reference=@n25, Delegate=@n24) {is=title K2Node_AddDelegate}
   > UpdateInteractionKeybind()
   > Set ReadyToShow(ReadyToShow=true)
   Event PreConstruct()
   > SetLineMeshMaterial!(
    execute
    self:Athena Marker Pointer Object Reference=Get MarkerPointer() -> (MarkerPointer:Athena Marker Pointer Object Reference)
    Material:Material Interface Object Reference=/Game/Athena/HUD/Art/Materials/M_UI_ModalRadialExpand_PointerVersion
   ) -> (then)
  ]}
 IsInteractionWidgetSurpressed {nodes [
   IsInteractionWidgetSurpressed() {is=title K2Node_FunctionEntry}
   > n1 = Is Valid(InputObject=Get CurrentInteraction() -> (CurrentInteraction:Fort Interact Context Info Object Reference))
   Return Node!(execute=@n1.Is Not Valid, bIsSurpressed:bool=false)
   > Return Node(bIsSurpressed=Get bSuppressInteractionWidget!(self:Fort Interact Context Info Object Reference=Get CurrentInteraction()) -> (bSuppressInteractionWidget:bool))
  ]}
 SetupInteraction {nodes [
   SetupInteraction() {is=title K2Node_FunctionEntry}
   n1 = Get ContextText!(self:Fort Interact Context Info Object Reference=GetActiveContextInfo()) -> (ContextText:Text)
   n2 = Get InteractText() -> (InteractText:Text Object Reference)
   > SetVisibility(self=@n2, InVisibility=Select(Option 0=SelfHitTestInvisible, Option 1=Collapsed, Index=TextIsEmpty(InText=@n1)))
   > "SetText (Text)"(self=@n2, InText=TextToUpper!(InText:"Text (by ref)"=@n1) -> (ReturnValue:Text))
   n3 = GetActiveContextInfo()
   n4 = Get bDisplayTextOnly!(self:Fort Interact Context Info Object Reference=@n3) -> (bDisplayTextOnly:bool)
   n5 = TextIsEmpty(InText=Get ContextText(self=@n3))
   > Set bKeybindWidgetVisible(bKeybindWidgetVisible:bool=AND Boolean(A=NOT Boolean(A=@n4), B=NOT Boolean(A=@n5))) -> (Output_Get:bool)
   > SetVisibility(
    self=Get KeybindWidget() -> (KeybindWidget:Keybind Widget Object Reference)
    InVisibility=Select(Option 0=Collapsed, Option 1=SelfHitTestInvisible, Index=Get bKeybindWidgetVisible() -> (bKeybindWidgetVisible:bool))
   )
   > Branch(Condition=Get bKeybindWidgetVisible())
   n6 = GetActiveContextInfo()
   n7 = NOT Boolean(A=Get bDisplayTextOnly(self=@n6))
   n8 = "float > float"(A=Get RequiredDuration!(self:Fort Interact Context Info Object Reference=@n6) -> (RequiredDuration:float), B=0) {is=title K2Node_CallFunction}
   > SetForcedHoldKeybindStatus!(
    execute
    self:Fort Keybind Widget Object Reference=Get KeybindWidget()
    InForcedHoldKeybindStatus:EFortKeybindForcedHoldStatus Enum=Select(
     Option 0:EFortKeybindForcedHoldStatus Enum=NoForcedHold
     Option 1:EFortKeybindForcedHoldStatus Enum=ForcedHold
     Index=AND Boolean(A=@n8, B=@n7)
    ) -> (ReturnValue:EFortKeybindForcedHoldStatus Enum)
   ) -> (then)
   n9 = GetActiveContextInfo()
   n10 = Get bDisplayTextOnly(self=@n9)
   n11 = Get RequiredDuration(self=@n9)
   > SetShowProgressCountDown!(
    execute
    self:Fort Keybind Widget Object Reference=Get KeybindWidget()
    bShow:bool=AND Boolean!(
     A="float > float"(A=@n11, B=0)
     B=Get bShowCountDown!(self:Fort Interact Context Info Object Reference=@n9) -> (bShowCountDown:bool)
     C:bool=NOT Boolean(A=@n10)
    ) -> (ReturnValue)
   ) -> (then)
  ]}
 ShowBasicInteractionWidget {nodes [
   ShowBasicInteractionWidget() {is=title K2Node_FunctionEntry}
   > UpdateInteractionKeybind()
   n1 = GetOwningPlayer()
   n2 = GetActiveContextInfo()
   > Branch(Condition=AND Boolean!(
     A=IsUsingTouch(PlayerController=@n1)
     B=Get bSuppressSimpleInteractionWidgetForTouch!(self:Fort Interact Context Info Object Reference=@n2) -> (bSuppressSimpleInteractionWidgetForTouch:bool)
    ) -> (ReturnValue))
   > SetVisibility(InVisibility=Collapsed)
   > SetVisibility(self=Get InteractText(), InVisibility=Collapsed)
  ]}
 UpdateInteractionKeybind {nodes [
   UpdateInteractionKeybind() {is=title K2Node_FunctionEntry}
   n1 = Get KeybindWidget()
   > SetBoundAction!(
    execute
    self:Fort Keybind Widget Object Reference=@n1
    NewBoundAction:Name=GetUseActionName!(
     self:Fort Global UIContext Object Reference=Get FortGlobalUIContext()
     OverrideInputType:ECommonInputType Enum=Get InputTypeOverride!(self:Fort Keybind Widget Object Reference=@n1) -> (InputTypeOverride:ECommonInputType Enum)
    ) -> (ReturnValue:Name=None)
   ) -> (then)
  ]}
 UpdateWidgetVisibility {nodes [
   UpdateWidgetVisibility() {is=title K2Node_FunctionEntry}
   > SetVisibility(InVisibility=Select(
     Option 0:ESlateVisibility Enum=SelfHitTestInvisible
     Option 1:ESlateVisibility Enum=Collapsed
     Index=IsInteractionWidgetSurpressed!() -> (bIsSurpressed:bool) {is=title K2Node_CallFunction}
    ) -> (ReturnValue:ESlateVisibility Enum))
  ]}
]
variables [
 CurReceivedActor {type=Actor, default=null}
 CurrentInteraction {type=FortInteractContextInfo, default=null}
 ReadyToShow {type=boolean}
 ValidInteraction {type=boolean}
]
widgets {
 info {inheritedWidgetCount=3, parentClass=/Script/FortniteUI.FortPortalHUDDetails, rootWidgetClass=&SizeBox, widgetCount=18}
 widgets [
  SizeBox_0 {
   widgetClassPath=&SizeBox
   widget=SizeBox_0
   properties {bOverride_HeightOverride=true, bOverride_WidthOverride=true, heightOverride=1080, widthOverride=1920}
  }
  CanvasPanel_1 {widgetClassPath=&CanvasPanel, parent=SizeBox_0, widget=CanvasPanel_1, slot {class=SizeBoxSlot}}
  Overlay_0 {
   widgetClassPath=&Overlay
   parent=CanvasPanel_1
   widget=Overlay_0
   slot {
    class=CanvasPanelSlot
    properties {bAutoSize=true, layoutData {alignment=<0 0>, anchors {maximum=<0.5 0.5>, minimum=<0.5 0.5>}, offsets=<ltrb 154.7338 -275 100 6.5>}}
   }
  }
  Image_Backplate {
   widgetClassPath=&Image
   bIsVariable=true
   parent=Overlay_0
   widget=Image_Backplate
   properties {
    renderOpacity=0
    visibility=SelfHitTestInvisible
    brush {
     bIsDynamicallyLoaded=false
     drawAs=Image
     imageType=NoImage
     mirroring=NoMirror
     resourceName=None
     resourceObject=/Game/Playgrounds/Items/Portals/UI/Art/M_SimpleSkewBox_Cr_PortalCalloutBacking
     tiling=NoTile
     imageSize=<32 32>
     margin=<ltrb 0 0 0 0>
     outlineSettings {
      bUseBrushTransparency=false
      roundingType=HalfHeightRadius
      width=0
      color {colorUseRule=UseColor_Specified, specifiedColor=<rgba 0 0 0 0>}
      cornerRadii=<0 0 0 0>
     }
     tintColor {colorUseRule=UseColor_Specified, specifiedColor=<rgba 1 1 1 1>}
     uVRegion {bIsValid=false, max=<0 0>, min=<0 0>}
    }
    colorAndOpacity=<rgba 0 0.0062 0.0677 0.9>
    renderTransform {angle=0, scale=<1.45 1.5>, shear=<0 0>, translation=<0 0>}
   }
   slot {class=OverlaySlot, properties {horizontalAlignment=HAlign_Fill, verticalAlignment=VAlign_Fill}}
  }
  Image_Glow_1 {
   widgetClassPath=&Image
   parent=Overlay_0
   widget=Image_Glow_1
   properties {
    renderOpacity=0
    visibility=SelfHitTestInvisible
    brush {
     bIsDynamicallyLoaded=false
     drawAs=Image
     imageType=NoImage
     mirroring=NoMirror
     resourceName=None
     resourceObject=/Game/Athena/HUD/Materials/UI_RebootAlert_Glow_Inst
     tiling=NoTile
     imageSize=<32 32>
     margin=<ltrb 0 0 0 0>
     outlineSettings {
      bUseBrushTransparency=false
      roundingType=HalfHeightRadius
      width=0
      color {colorUseRule=UseColor_Specified, specifiedColor=<rgba 0 0 0 0>}
      cornerRadii=<0 0 0 0>
     }
     tintColor {colorUseRule=UseColor_Specified, specifiedColor=<rgba 1 1 1 1>}
     uVRegion {bIsValid=false, max=<0 0>, min=<0 0>}
    }
    colorAndOpacity=<rgba 0 0.0138 0.151 1>
    renderTransform {angle=0, scale=<2.5254 2.4649>, shear=<0 0>, translation=<0 0>}
   }
   slot {class=OverlaySlot, properties {horizontalAlignment=HAlign_Fill, verticalAlignment=VAlign_Fill}}
  }
  Image_Glow_2 {
   widgetClassPath=&Image
   parent=Overlay_0
   widget=Image_Glow_2
   properties {
    visibility=Collapsed
    brush {
     bIsDynamicallyLoaded=false
     drawAs=Image
     imageType=NoImage
     mirroring=NoMirror
     resourceName=None
     resourceObject=&UI_RebootAlert_Glow_Inst
     tiling=NoTile
     imageSize=<32 32>
     margin=<ltrb 0 0 0 0>
     outlineSettings {
      bUseBrushTransparency=false
      roundingType=HalfHeightRadius
      width=0
      color {colorUseRule=UseColor_Specified, specifiedColor=<rgba 0 0 0 0>}
      cornerRadii=<0 0 0 0>
     }
     tintColor {colorUseRule=UseColor_Specified, specifiedColor=<rgba 1 1 1 1>}
     uVRegion {bIsValid=false, max=<0 0>, min=<0 0>}
    }
    colorAndOpacity=<rgba 0 0.0908 0.9911 1>
    renderTransform {angle=0, scale=<2.5254 2.4649>, shear=<0 0>, translation=<0 0>}
   }
   slot {class=OverlaySlot, properties {horizontalAlignment=HAlign_Fill, verticalAlignment=VAlign_Fill}}
  }
  VerticalBox_0 {widgetClassPath=&VerticalBox, parent=Overlay_0, widget=VerticalBox_0, slot {class=OverlaySlot}}
  Border_97 {
   widgetClassPath=&Border
   parent=VerticalBox_0
   widget=Border_97
   properties {
    visibility=SelfHitTestInvisible
    background {
     bIsDynamicallyLoaded=false
     drawAs=NoDrawType
     imageType=NoImage
     mirroring=NoMirror
     resourceName=None
     resourceObject=null
     tiling=NoTile
     imageSize=<32 32>
     margin=<ltrb 0 0 0 0>
     outlineSettings {
      bUseBrushTransparency=false
      roundingType=HalfHeightRadius
      width=0
      color {colorUseRule=UseColor_Specified, specifiedColor=<rgba 0 0 0 0>}
      cornerRadii=<0 0 0 0>
     }
     tintColor {colorUseRule=UseColor_Specified, specifiedColor=<rgba 1 1 1 1>}
     uVRegion {bIsValid=false, max=<0 0>, min=<0 0>}
    }
    brushColor=<rgba 1 1 1 0>
    padding=<ltrb 0 0 0 0>
   }
   slot {class=VerticalBoxSlot}
  }
  Overlay_4 {widgetClassPath=&Overlay, parent=Border_97, widget=Overlay_4, slot {class=BorderSlot, properties {padding=<ltrb 0 0 0 0>}}}
  VertBox_Content {widgetClassPath=&VerticalBox, bIsVariable=true, parent=Overlay_4, widget=VertBox_Content, slot {class=OverlaySlot}}
  Overlay_1 {
   widgetClassPath=&Overlay
   parent=VerticalBox_0
   widget=Overlay_1
   slot {class=VerticalBoxSlot, properties {verticalAlignment=VAlign_Top}}
  }
  MarkerPointer {
   widgetClassPath=/Script/FortniteUI.AthenaMarkerPointer
   bInherited=true
   bIsVariable=true
   parent=Overlay_1
   widget=MarkerPointer
   properties {renderTransform {angle=0, scale=<1 1>, shear=<0 0>, translation=<3 -3>}}
   slot {class=OverlaySlot}
  }
  Border_219 {
   widgetClassPath=&Border
   parent=Overlay_1
   widget=Border_219
   properties {brushColor=<rgba 0 0 0 0>}
   slot {class=OverlaySlot, properties {horizontalAlignment=HAlign_Fill, verticalAlignment=VAlign_Fill}}
  }
  HorizontalBox_0 {
   widgetClassPath=&HorizontalBox
   parent=Overlay_1
   widget=HorizontalBox_0
   slot {class=OverlaySlot, properties {padding=<ltrb 0 12 0 0>}}
  }
  KeybindWidget {
   widgetClassPath=/FortUILibrary/Input/KeybindWidget
   bIsVariable=true
   parent=HorizontalBox_0
   widget=KeybindWidget
   properties {renderTransform {angle=0, scale=<0.6 0.6>, shear=<0 0>, translation=<0 0>}}
   slot {class=HorizontalBoxSlot, properties {horizontalAlignment=HAlign_Left, verticalAlignment=VAlign_Center, padding=<ltrb 0 -10 0 -6>}}
  }
  InteractText {
   widgetClassPath=&TextBlock
   bIsVariable=true
   parent=HorizontalBox_0
   widget=InteractText
   properties {
    text=CHANGE DESTINATION
    font {
     bForceMonospaced=false
     bMaterialIsStencil=false
     fontMaterial=null
     fontObject=&BurbankSmall
     letterSpacing=0
     monospacedWidth=1
     size=24
     skewAmount=0
     typefaceFontName=Black
     outlineSettings {
      bApplyOutlineToDropShadows=false
      bMiteredCorners=false
      bSeparateFillAlpha=false
      outlineMaterial=null
      outlineSize=1
      outlineColor=<rgba 0.0052 0.0052 0.0056 0.5>
     }
    }
    renderTransform {angle=0, scale=<1 1>, shear=<-8 0>, translation=<0 0>}
    shadowColorAndOpacity=<rgba 0 0 0 0.5>
    shadowOffset=<3 1>
   }
   slot {
    class=HorizontalBoxSlot
    properties {horizontalAlignment=HAlign_Left, verticalAlignment=VAlign_Center, padding=<ltrb -4 -2 0 0>, size {sizeRule=Fill, value=1}}
   }
  }
  MobileInfo_Matchmaking_Switcher {widgetClassPath=&CommonActivatableWidgetSwitcher, bInherited=true, widget=null}
  TutorialHighlight {widgetClassPath=&TutorialHighlight}
 ]
}
