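# Creative matchmaking portal widget that shows a destination island's link code, creator, description and player range, or a loading or error state.
# Update Island Settings pulls details from the owning object through interface calls, fills the text and switches the main panel by portal state.
# A second switcher shows the local player's interaction status.
# For timed override locks, a looping one-second timer updates the countdown, then refreshes that status on expiry.
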WidgetBlueprint /Game/Creative/Devices/MatchmakingPortal/CreativeIslandDetails
animations [
 Intro {
  binding_count=2
  length_seconds=0.75
  object=&self:Intro
  sequence {bindings [
    Text_CreatorName {
     id=#1
     possessed_class=CommonTextBlock
     tracks [
      {
       type=MovieScene2DTransformTrack
       sections [{
         type=MovieScene2DTransformSection
         end_frame=5
         start_frame=0
         channels [
          Translation.X {type=MovieSceneScriptingFloatChannel}
          Translation.Y {type=MovieSceneScriptingFloatChannel}
          Angle {type=MovieSceneScriptingFloatChannel}
          Scale.X {type=MovieSceneScriptingFloatChannel, keys [[0,0.7056],[2,1.1],[3,0.95],[5,1]]}
          Scale.Y {type=MovieSceneScriptingFloatChannel, keys [[0,0.7056],[2,1.1],[3,0.95],[5,1]]}
          Shear.X {type=MovieSceneScriptingFloatChannel}
          Shear.Y {type=MovieSceneScriptingFloatChannel}
         ]
        }]
      }
      {
       type=MovieSceneFloatTrack
       display_name=Render Opacity
       sections [{type=MovieSceneFloatSection, end_frame=5, start_frame=0, channels [None {type=MovieSceneScriptingFloatChannel, keys [[0,0],[2,1],[5,1]]}]}]
      }
     ]
    }
    PopulationBox {
     id=#2
     possessed_class=HorizontalBox
     tracks [
      {
       type=MovieScene2DTransformTrack
       sections [{
         type=MovieScene2DTransformSection
         end_frame=15
         start_frame=8
         channels [
          Translation.X {type=MovieSceneScriptingFloatChannel, const [0,4]}
          Translation.Y {type=MovieSceneScriptingFloatChannel, keys [[8,-9],[11,3.6452],[13,-1.5518],[15,0]]}
          Angle {type=MovieSceneScriptingFloatChannel}
          Scale.X {type=MovieSceneScriptingFloatChannel}
          Scale.Y {type=MovieSceneScriptingFloatChannel}
          Shear.X {type=MovieSceneScriptingFloatChannel}
          Shear.Y {type=MovieSceneScriptingFloatChannel}
         ]
        }]
      }
      {
       type=MovieSceneFloatTrack
       display_name=Render Opacity
       sections [{type=MovieSceneFloatSection, end_frame=10, start_frame=0, channels [None {type=MovieSceneScriptingFloatChannel, keys [[0,0],[8,0],[10,1]]}]}]
      }
     ]
    }
   ]}
 }
 intro2 {
  binding_count=2
  length_seconds=0.15
  object=&self:intro2
  sequence {bindings [
    Text_CreatorName {
     id=#3
     possessed_class=CommonTextBlock
     tracks [
      {
       type=MovieScene2DTransformTrack
       sections [{
         type=MovieScene2DTransformSection
         end_frame=3
         start_frame=0
         channels [
          Translation.X {type=MovieSceneScriptingFloatChannel, keys [[0,-48],[3,0]]}
          Translation.Y {type=MovieSceneScriptingFloatChannel, const [0,2]}
          Angle {type=MovieSceneScriptingFloatChannel}
          Scale.X {type=MovieSceneScriptingFloatChannel}
          Scale.Y {type=MovieSceneScriptingFloatChannel}
          Shear.X {type=MovieSceneScriptingFloatChannel}
          Shear.Y {type=MovieSceneScriptingFloatChannel}
         ]
        }]
      }
      {
       type=MovieSceneFloatTrack
       display_name=Render Opacity
       sections [{type=MovieSceneFloatSection, end_frame=3, start_frame=0, channels [None {type=MovieSceneScriptingFloatChannel, keys [[0,0],[3,1]]}]}]
      }
     ]
    }
    PopulationBox {
     id=#4
     possessed_class=HorizontalBox
     tracks [
      {
       type=MovieScene2DTransformTrack
       sections [{
         type=MovieScene2DTransformSection
         end_frame=3
         start_frame=0
         channels [
          Translation.X {type=MovieSceneScriptingFloatChannel, keys [[0,-50],[3,0]]}
          Translation.Y {type=MovieSceneScriptingFloatChannel, const [0,2]}
          Angle {type=MovieSceneScriptingFloatChannel}
          Scale.X {type=MovieSceneScriptingFloatChannel}
          Scale.Y {type=MovieSceneScriptingFloatChannel}
          Shear.X {type=MovieSceneScriptingFloatChannel}
          Shear.Y {type=MovieSceneScriptingFloatChannel}
         ]
        }]
      }
      {
       type=MovieSceneFloatTrack
       display_name=Render Opacity
       sections [{type=MovieSceneFloatSection, end_frame=3, start_frame=0, channels [None {type=MovieSceneScriptingFloatChannel, keys [[0,0],[3,1]]}]}]
      }
     ]
    }
   ]}
 }
 DescriptionReveal {
  binding_count=1
  length_seconds=0.25
  object=&self:DescriptionReveal
  sequence {bindings [Text_Description {
     id=#5
     possessed_class=CommonTextBlock
     tracks [
      {
       type=MovieScene2DTransformTrack
       sections [{
         type=MovieScene2DTransformSection
         end_frame=5
         start_frame=0
         channels [
          Translation.X {type=MovieSceneScriptingFloatChannel}
          Translation.Y {type=MovieSceneScriptingFloatChannel}
          Angle {type=MovieSceneScriptingFloatChannel}
          Scale.X {type=MovieSceneScriptingFloatChannel, keys [[0,1.2424],[5,1]]}
          Scale.Y {type=MovieSceneScriptingFloatChannel, keys [[0,1.2424],[5,1]]}
          Shear.X {type=MovieSceneScriptingFloatChannel}
          Shear.Y {type=MovieSceneScriptingFloatChannel}
         ]
        }]
      }
      {
       type=MovieSceneFloatTrack
       display_name=Render Opacity
       sections [{type=MovieSceneFloatSection, end_frame=5, start_frame=0, channels [None {type=MovieSceneScriptingFloatChannel, keys [[0,0],[3,1],[5,1]]}]}]
      }
     ]
    }]}
 }
 VisibilityReveal {
  binding_count=1
  length_seconds=0.45
  object=&self:VisibilityReveal
  sequence {bindings [Text_CreatorName {
     id=#6
     possessed_class=CommonTextBlock
     tracks [
      {
       type=MovieSceneWidgetMaterialTrack
       display_name=Font.Outline Settings.OutlineMaterial
       sections [{type=MovieSceneParameterSection, end_frame=9, start_frame=0, channels [Position {type=MovieSceneScriptingFloatChannel, keys [[0,0],[9,1]]}]}]
      }
      {
       type=MovieSceneFloatTrack
       display_name=Render Opacity
       sections [{
         type=MovieSceneFloatSection
         end_frame=9
         start_frame=0
         channels [None {type=MovieSceneScriptingFloatChannel, keys [[0,0],[5,1],[9.0003,1]]}]
        }]
      }
      {
       type=MovieScene2DTransformTrack
       sections [{
         type=MovieScene2DTransformSection
         end_frame=9
         start_frame=0
         channels [
          Translation.X {type=MovieSceneScriptingFloatChannel}
          Translation.Y {type=MovieSceneScriptingFloatChannel}
          Angle {type=MovieSceneScriptingFloatChannel}
          Scale.X {type=MovieSceneScriptingFloatChannel, keys [[0,0.6841],[5,1.1],[7,0.9646],[9,1]]}
          Scale.Y {type=MovieSceneScriptingFloatChannel, const [1,4]}
          Shear.X {type=MovieSceneScriptingFloatChannel}
          Shear.Y {type=MovieSceneScriptingFloatChannel}
         ]
        }]
      }
     ]
    }]}
 }
]
graphs [
 EventGraph {nodes [
   Check Lock Timer() {is=title K2Node_CustomEvent}
   > n1 = Update Timer!(execute, self) -> (then, Has Time Remaining:bool)
   > n2 = Branch(Condition=@n1)
   n3 = Looping Timer Update() {is=title K2Node_CustomEvent}
   n4 = Set Timer by Event(execute=@n2.then, Delegate=@n3, Time=1, bLooping=true)
   > Set Timer Counter(Timer Counter:Timer Handle Structure=@n4) -> (Output_Get:Timer Handle Structure)
   Play Valid Details Intro Anims() {is=title K2Node_CustomEvent}
   > PlayAnimation(InAnimation=Get DescriptionReveal() -> (DescriptionReveal:Widget Animation Object Reference))
   > PlayAnimation(InAnimation=Get intro2() -> (intro2:Widget Animation Object Reference))
   Update Interaction Details(execute=@n6.then)
   n5 = Update Timer(execute=@n3.then)
   > n6 = Branch(Condition=NOT Boolean(A=@n5))
  ]}
 Set Island Widget Visual {nodes [
   n1 = Set Island Widget Visual!() -> (then, New State:ECreativeMatchmakingPortalState Enum=NewEnumerator0) {is=title K2Node_FunctionEntry}
   > SetActiveWidgetIndex(
    self=Get MainDetails() -> (MainDetails:Common Animated Switcher Object Reference)
    Index="To Integer (Byte)"!(InByte:Byte=@n1) -> (ReturnValue:int=0)
   )
   > SetVisibility(InVisibility=HitTestInvisible)
   > Update Interaction Details()
  ]}
 Update Interaction Details {nodes [
   Update Interaction Details() {is=title K2Node_FunctionEntry}
   > n1 = Get!(execute:Exec) -> (then:Exec, else:Exec, Owning Object:Object Reference)
   > n2 = Cast To FortPlayerController(Object=GetOwningPlayer())
   > n3 = Get Interaction Details!(execute:Exec, self:Object Reference=@n1, Controller:Fort Player Controller Object Reference=@n2) -> (then:Exec, Interaction State:ECreativeMatchmakingPortalInteraction Enum=NewEnumerator0, Remaining Lock Time:Timespan Structure) {is=title K2Node_Message}
   > SetActiveWidgetIndex(
    self=Get InteractDetails() -> (InteractDetails:Common Widget Switcher Legacy Object Reference)
    Index="To Integer (Byte)"(InByte=@n3) -> (ReturnValue=0)
   )
   > Branch(Condition="Equal (Enum)"(A:ECreativeMatchmakingPortalInteraction Enum=@n3, B:ECreativeMatchmakingPortalInteraction Enum=NewEnumerator5))
   > Check Lock Timer() {is=title K2Node_CallFunction}
  ]}
 Update Island Settings {nodes [
   n1 = Update Island Settings!() -> (then, Owning Object:Object Reference) {is=title K2Node_FunctionEntry}
   > n2 = Set Owning Object(Owning Object:Object Reference=@n1) -> (Output_Get:Object Reference)
   > n3 = Cast To Creative_Device_Prop_Parent!(execute, Object=@n2) -> (then, CastFailed, AsCreative Device Prop Parent:Creative Device Prop Parent Object Reference) {is=title K2Node_DynamicCast}
   > n4 = Bind Event to On Device Activated(self:Creative Device Prop Parent Object Reference=@n3, Delegate=Create Event()) {is=title K2Node_AddDelegate}
   n5 = Get IslandPortal Details!(execute:Exec=[@n4.then, @n3.CastFailed], self:Object Reference=Get Owning Object() -> (Owning Object:Object Reference)) -> (
    then:Exec
    Matchmaking State:ECreativeMatchmakingPortalState Enum=NewEnumerator0
    Creator Text:Text
    Description Text:Text
    Link Text:Text
    Pop Min:int=0
    Pop Max:int=0
   ) {is=title K2Node_Message}
   > "SetText (Text)"(self=Get Text_CreatorName() -> (Text_CreatorName:Common Text Object Reference), InText=@n5.Creator Text)
   > "SetText (Text)"(self=Get Text_Description() -> (Text_Description:Common Text Object Reference), InText=@n5.Description Text)
   > "SetText (Text)"(self=Get Text_LinkCode() -> (Text_LinkCode:Common Text Object Reference), InText=@n5.Link Text)
   > Update Population Text!(execute, self, Min:int=@n5.Pop Min, Max:int=@n5.Pop Max) -> (then)
   > Set Island Widget Visual!(execute, self, New State:ECreativeMatchmakingPortalState Enum=@n5) -> (then) {is=title K2Node_CallFunction}
   > Update Interaction Details() {is=title K2Node_CallFunction}
  ]}
 Update Population Text {nodes [
   n1 = Update Population Text!() -> (then, Min:int, Max:int) {is=title K2Node_FunctionEntry}
   n2 = "integer > integer"(A=@n1.Max, B=0) {is=title K2Node_CallFunction}
   > SetVisibility(
    self=[Get Text_IslandPop() -> (Text_IslandPop:Common Text Object Reference), Get SocialImage() -> (SocialImage:Image Object Reference)]
    InVisibility=Select(Option 0=Collapsed, Option 1=SelfHitTestInvisible, Index=@n2)
   )
   > Branch(Condition=@n2)
   > "SetText (Text)"(
    self=Get Text_IslandPop()
    InText=Select(
     Option 0:Text=Format Text!(Format="LOCTEXT(\"{Min} - {Max}\")", Min:int=@n1, Max:int=@n1.Max) -> (Result)
     Option 1:Text="To Text (Integer)"(Value=@n1.Max)
     Index="Equal (Integer)"(A=@n1, B=@n1.Max) {is=title K2Node_CallFunction}
    ) -> (ReturnValue:Text)
   )
  ]}
 Update Timer {nodes [
   Update Timer() {is=title K2Node_FunctionEntry}
   > n1 = Get()
   > n2 = Cast To FortPlayerController(Object=GetOwningPlayer())
   > n3 = Get Interaction Details(self=@n1, Controller=@n2) -> (Interaction State=NewEnumerator0)
   > Set Lock Time(Lock Time:Timespan Structure=@n3.Remaining Lock Time) -> (Output_Get:Timespan Structure)
   n4 = Get Lock Time() -> (Lock Time:Timespan Structure)
   n5 = Get Minutes!(A:Timespan Structure=@n4) -> (ReturnValue:int=0) {is=title K2Node_CallFunction}
   n6 = Get Seconds!(A:Timespan Structure=@n4) -> (ReturnValue:int=0) {is=title K2Node_CallFunction}
   > "SetText (Text)"(
    self=Get Timer() -> (Timer:Common Text Object Reference)
    InText=Select(
     Option 0=Format Text!(Format="LOCTEXT(\"{time}s\")", time:int=@n6) -> (Result)
     Option 1=Format Text(Format="LOCTEXT(\"{time}m\")", time=@n5)
     Index="integer >= integer"!(A:int=@n5, B:int=1) -> (ReturnValue:bool) {is=title K2Node_CallFunction}
    )
   )
   n7 = Truncate(A=Get Total Seconds!(A:Timespan Structure=Get Lock Time()) -> (ReturnValue:double) {is=title K2Node_CallFunction})
   > n8 = Branch(Condition=OR Boolean(
     A="integer < integer"!(A:int=@n7, B:int=1) -> (ReturnValue:bool) {is=title K2Node_CallFunction}
     B="Not Equal (Enum)"!(A:ECreativeMatchmakingPortalInteraction Enum=@n3, B:ECreativeMatchmakingPortalInteraction Enum=NewEnumerator5) -> (ReturnValue:bool)
    ))
   > Clear and Invalidate Timer by Handle(Handle=Get Timer Counter() -> (Timer Counter:Timer Handle Structure))
   > Return Node!(execute, Has Time Remaining:bool=false)
   Return Node(execute=@n8.else, Has Time Remaining=true)
  ]}
]
variables [
 Owning Object {type=Object, default=null}
 Timer Counter {type=TimerHandle}
]
widgets {
 info {inheritedWidgetCount=0, parentClass=&UserWidget, rootWidgetClass=&Overlay, widgetCount=31}
 widgets [
  Overlay_2 {widgetClassPath=&Overlay, widget=Overlay_2}
  VerticalBox_89 {widgetClassPath=&VerticalBox, parent=Overlay_2, widget=VerticalBox_89, slot {class=OverlaySlot}}
  MainDetails {
   widgetClassPath=/Script/CommonUI.CommonAnimatedSwitcher
   bIsVariable=true
   parent=VerticalBox_89
   widget=MainDetails
   properties {activeWidgetIndex=3}
   slot {class=VerticalBoxSlot, properties {horizontalAlignment=HAlign_Left, verticalAlignment=VAlign_Top}}
  }
  Hidden {widgetClassPath=&Spacer, parent=MainDetails, widget=Hidden, slot {class=WidgetSwitcherSlot}}
  LoadingDestination {
   widgetClassPath=&HorizontalBox
   parent=MainDetails
   widget=LoadingDestination
   slot {class=WidgetSwitcherSlot, properties {padding=<ltrb 0 0 100 0>}}
  }
  Throbber {
   widgetClassPath=&Image
   bIsVariable=true
   parent=LoadingDestination
   widget=Throbber
   properties {
    visibility=HitTestInvisible
    brush {
     bIsDynamicallyLoaded=false
     drawAs=Image
     imageType=NoImage
     mirroring=NoMirror
     resourceName=None
     resourceObject=/FortUILibrary/UIKit/Throbber/MI_UIKit_Monochrome_Throbber
     tiling=NoTile
     imageSize=<32 32>
     margin=<ltrb 0 0 0 0>
     outlineSettings {
      bUseBrushTransparency=false
      roundingType=HalfHeightRadius
      width=0
      color {colorUseRule=UseColor_Specified, specifiedColor=<rgba 0 0 0 0>}
      cornerRadii=<0 0 0 0>
     }
     tintColor {colorUseRule=UseColor_Specified, specifiedColor=<rgba 1 1 1 1>}
     uVRegion {bIsValid=false, max=<0 0>, min=<0 0>}
    }
   }
   slot {class=HorizontalBoxSlot}
  }
  Text_Loading {
   widgetClassPath=&CommonTextBlock
   parent=LoadingDestination
   widget=Text_Loading
   properties {
    justification=Center
    text=LOADING DESTINATION
    visibility=HitTestInvisible
    colorAndOpacity {colorUseRule=UseColor_Specified, specifiedColor=<rgba 0.6108 0.6406 0.6274 1>}
    font {
     bForceMonospaced=false
     bMaterialIsStencil=false
     fontMaterial=null
     fontObject=&BurbankSmall
     letterSpacing=0
     monospacedWidth=1
     size=16
     skewAmount=0
     typefaceFontName=Black
     outlineSettings {
      bApplyOutlineToDropShadows=false
      bMiteredCorners=false
      bSeparateFillAlpha=false
      outlineMaterial=null
      outlineSize=1
      outlineColor=<rgba 0 0.066 0.066 0.3>
     }
    }
    renderTransform {angle=0, scale=<1 1>, shear=<-8 0>, translation=<0 0>}
    shadowColorAndOpacity=<rgba 0.0052 0.0052 0.0056 0.822>
    shadowOffset=<0 0>
   }
   slot {class=HorizontalBoxSlot, properties {horizontalAlignment=HAlign_Left, verticalAlignment=VAlign_Bottom, padding=<ltrb 4 0 0 0>}}
  }
  InvalidDestination {widgetClassPath=&HorizontalBox, parent=MainDetails, widget=InvalidDestination, slot {class=WidgetSwitcherSlot}}
  Text_Invalid {
   widgetClassPath=&CommonTextBlock
   parent=InvalidDestination
   widget=Text_Invalid
   properties {
    justification=Center
    text=Invalid code
    visibility=HitTestInvisible
    colorAndOpacity {colorUseRule=UseColor_Specified, specifiedColor=<rgba 1 0.0622 0 1>}
    font {
     bForceMonospaced=false
     bMaterialIsStencil=false
     fontMaterial=null
     fontObject=&BurbankSmall
     letterSpacing=0
     monospacedWidth=1
     size=16
     skewAmount=0
     typefaceFontName=Black
     outlineSettings {
      bApplyOutlineToDropShadows=false
      bMiteredCorners=false
      bSeparateFillAlpha=false
      outlineMaterial=null
      outlineSize=1
      outlineColor=<rgba 0 0.066 0.066 0.3>
     }
    }
    renderTransform {angle=0, scale=<1 1>, shear=<-8 0>, translation=<0 0>}
    shadowColorAndOpacity=<rgba 0.0052 0.0052 0.0056 0.822>
    shadowOffset=<0 0>
   }
   slot {class=HorizontalBoxSlot, properties {horizontalAlignment=HAlign_Left, verticalAlignment=VAlign_Bottom}}
  }
  LoadedIslandDescription {
   widgetClassPath=&VerticalBox
   parent=MainDetails
   widget=LoadedIslandDescription
   slot {class=WidgetSwitcherSlot, properties {horizontalAlignment=HAlign_Left}}
  }
  Text_LinkCode {
   widgetClassPath=&CommonTextBlock
   bIsVariable=true
   parent=LoadedIslandDescription
   widget=Text_LinkCode
   properties {
    justification=Center
    text=4536-7842-2454
    font {
     bForceMonospaced=false
     bMaterialIsStencil=false
     fontMaterial=null
     fontObject=&BurbankSmall
     letterSpacing=0
     monospacedWidth=1
     size=18
     skewAmount=0
     typefaceFontName=Black
     outlineSettings {
      bApplyOutlineToDropShadows=false
      bMiteredCorners=false
      bSeparateFillAlpha=false
      outlineMaterial=null
      outlineSize=1
      outlineColor=<rgba 0.0052 0.0052 0.0056 0.2>
     }
    }
    renderTransform {angle=0, scale=<1 1>, shear=<-8 0>, translation=<0 0>}
    shadowColorAndOpacity=<rgba 0.0052 0.0052 0.0056 0.822>
    shadowOffset=<0 0>
   }
   slot {class=VerticalBoxSlot, properties {horizontalAlignment=HAlign_Left}}
  }
  CreatedByBox {widgetClassPath=&HorizontalBox, parent=LoadedIslandDescription, widget=CreatedByBox, slot {class=VerticalBoxSlot}}
  CommonTextBlock_75 {
   widgetClassPath=&CommonTextBlock
   parent=CreatedByBox
   widget=CommonTextBlock_75
   properties {
    text="Created by: "
    textTransformPolicy=ToUpper
    colorAndOpacity {colorUseRule=UseColor_Specified, specifiedColor=<rgba 1 1 1 0.8>}
    font {
     bForceMonospaced=false
     bMaterialIsStencil=false
     fontMaterial=null
     fontObject=&BurbankSmall
     letterSpacing=0
     monospacedWidth=1
     size=14
     skewAmount=0
     typefaceFontName=Black
     outlineSettings {
      bApplyOutlineToDropShadows=false
      bMiteredCorners=false
      bSeparateFillAlpha=false
      outlineMaterial=null
      outlineSize=1
      outlineColor=<rgba 0.0052 0.0052 0.0056 0.149>
     }
    }
    renderTransform {angle=0, scale=<1 1>, shear=<-8 0>, translation=<0 0>}
    shadowOffset=<0 0>
   }
   slot {class=HorizontalBoxSlot, properties {verticalAlignment=VAlign_Bottom}}
  }
  Text_CreatorName {
   widgetClassPath=&CommonTextBlock
   bIsVariable=true
   parent=CreatedByBox
   widget=Text_CreatorName
   properties {
    justification=Center
    text=LONGNAMEPERSON
    visibility=HitTestInvisible
    colorAndOpacity {colorUseRule=UseColor_Specified, specifiedColor=<rgba 0 1 1 1>}
    font {
     bForceMonospaced=false
     bMaterialIsStencil=false
     fontMaterial=null
     fontObject=&BurbankSmall
     letterSpacing=0
     monospacedWidth=1
     size=16
     skewAmount=0
     typefaceFontName=Black
     outlineSettings {
      bApplyOutlineToDropShadows=false
      bMiteredCorners=false
      bSeparateFillAlpha=false
      outlineMaterial=null
      outlineSize=1
      outlineColor=<rgba 0 0.066 0.066 0.3>
     }
    }
    renderTransform {angle=0, scale=<1 1>, shear=<-8 0>, translation=<0 0>}
    shadowColorAndOpacity=<rgba 0.0052 0.0052 0.0056 0.822>
    shadowOffset=<0 0>
   }
   slot {class=HorizontalBoxSlot, properties {verticalAlignment=VAlign_Bottom, padding=<ltrb 4 0 0 0>}}
  }
  PopulationBox {
   widgetClassPath=&HorizontalBox
   bIsVariable=true
   parent=CreatedByBox
   widget=PopulationBox
   slot {class=HorizontalBoxSlot, properties {horizontalAlignment=HAlign_Center, padding=<ltrb 20 0 10 0>}}
  }
  SocialImage {
   widgetClassPath=&Image
   bIsVariable=true
   parent=PopulationBox
   widget=SocialImage
   properties {brush {
     bIsDynamicallyLoaded=false
     drawAs=Image
     imageType=NoImage
     mirroring=NoMirror
     resourceName=None
     resourceObject=/Game/UI/Frontend/Parties/PartiesButtonStyles/T_UI_Icon_QuadMeeple
     tiling=NoTile
     imageSize=<20 20>
     margin=<ltrb 0 0 0 0>
     outlineSettings {
      bUseBrushTransparency=false
      roundingType=HalfHeightRadius
      width=0
      color {colorUseRule=UseColor_Specified, specifiedColor=<rgba 0 0 0 0>}
      cornerRadii=<0 0 0 0>
     }
     tintColor {colorUseRule=UseColor_Specified, specifiedColor=<rgba 1 1 1 1>}
     uVRegion {bIsValid=false, max=<0 0>, min=<0 0>}
    }}
   slot {class=HorizontalBoxSlot, properties {verticalAlignment=VAlign_Center, padding=<ltrb 0 0 6 0>}}
  }
  Text_IslandPop {
   widgetClassPath=&CommonTextBlock
   bIsVariable=true
   parent=PopulationBox
   widget=Text_IslandPop
   properties {
    text=00 - 00
    font {
     bForceMonospaced=false
     bMaterialIsStencil=false
     fontMaterial=null
     fontObject=&BurbankSmall
     letterSpacing=0
     monospacedWidth=1
     size=16
     skewAmount=0
     typefaceFontName=Black
     outlineSettings {
      bApplyOutlineToDropShadows=false
      bMiteredCorners=false
      bSeparateFillAlpha=false
      outlineMaterial=null
      outlineSize=1
      outlineColor=<rgba 0.0052 0.0052 0.0056 0.149>
     }
    }
    shadowColorAndOpacity=<rgba 0.0052 0.0052 0.0056 0.822>
    shadowOffset=<0 0>
   }
   slot {class=HorizontalBoxSlot, properties {verticalAlignment=VAlign_Bottom, padding=<ltrb 0 0 0 -2>}}
  }
  Text_Description {
   widgetClassPath=&CommonTextBlock
   bIsVariable=true
   parent=LoadedIslandDescription
   widget=Text_Description
   properties {
    bAutoCollapseWithEmptyText=true
    text=Suspiciously long vague description text about this that or the other WHO KNOWS
    wrapTextAt=465
    colorAndOpacity {colorUseRule=UseColor_Specified, specifiedColor=<rgba 0 1 1 1>}
    font {
     bForceMonospaced=false
     bMaterialIsStencil=false
     fontMaterial=null
     fontObject=&BurbankSmall
     letterSpacing=0
     monospacedWidth=1
     size=12
     skewAmount=0
     typefaceFontName=Bold
     outlineSettings {
      bApplyOutlineToDropShadows=false
      bMiteredCorners=false
      bSeparateFillAlpha=false
      outlineMaterial=null
      outlineSize=1
      outlineColor=<rgba 0.0052 0.0052 0.0056 0.149>
     }
    }
    shadowColorAndOpacity=<rgba 0.0052 0.0052 0.0056 0.822>
    shadowOffset=<0 0>
   }
   slot {class=VerticalBoxSlot, properties {horizontalAlignment=HAlign_Left, padding=<ltrb 12 0 0 0>}}
  }
  NoDestination {widgetClassPath=&HorizontalBox, parent=MainDetails, widget=NoDestination, slot {class=WidgetSwitcherSlot}}
  Text_NoDestination {
   widgetClassPath=&CommonTextBlock
   parent=NoDestination
   widget=Text_NoDestination
   properties {
    justification=Center
    text=No destination has been set
    colorAndOpacity {colorUseRule=UseColor_Specified, specifiedColor=<rgba 1 0.5638 0 1>}
    font {
     bForceMonospaced=false
     bMaterialIsStencil=false
     fontMaterial=null
     fontObject=&BurbankSmall
     letterSpacing=0
     monospacedWidth=1
     size=16
     skewAmount=0
     typefaceFontName=Black
     outlineSettings {
      bApplyOutlineToDropShadows=false
      bMiteredCorners=false
      bSeparateFillAlpha=false
      outlineMaterial=null
      outlineSize=1
      outlineColor=<rgba 0 0.066 0.066 0.3>
     }
    }
    renderTransform {angle=0, scale=<1 1>, shear=<-8 0>, translation=<0 0>}
    shadowColorAndOpacity=<rgba 0.0052 0.0052 0.0056 0.822>
    shadowOffset=<0 0>
   }
   slot {class=HorizontalBoxSlot, properties {horizontalAlignment=HAlign_Left, verticalAlignment=VAlign_Bottom, padding=<ltrb 4 0 0 0>}}
  }
  PrivateDestination {widgetClassPath=&HorizontalBox, parent=MainDetails, widget=PrivateDestination, slot {class=WidgetSwitcherSlot}}
  Text_PrivateDestination {
   widgetClassPath=&CommonTextBlock
   parent=PrivateDestination
   widget=Text_PrivateDestination
   properties {
    justification=Center
    text=The destination is private
    colorAndOpacity {colorUseRule=UseColor_Specified, specifiedColor=<rgba 1 0.5638 0 1>}
    font {
     bForceMonospaced=false
     bMaterialIsStencil=false
     fontMaterial=null
     fontObject=&BurbankSmall
     letterSpacing=0
     monospacedWidth=1
     size=16
     skewAmount=0
     typefaceFontName=Black
     outlineSettings {
      bApplyOutlineToDropShadows=false
      bMiteredCorners=false
      bSeparateFillAlpha=false
      outlineMaterial=null
      outlineSize=1
      outlineColor=<rgba 0 0.066 0.066 0.3>
     }
    }
    renderTransform {angle=0, scale=<1 1>, shear=<-8 0>, translation=<0 0>}
    shadowColorAndOpacity=<rgba 0.0052 0.0052 0.0056 0.822>
    shadowOffset=<0 0>
   }
   slot {class=HorizontalBoxSlot, properties {horizontalAlignment=HAlign_Left, verticalAlignment=VAlign_Bottom, padding=<ltrb 4 0 0 0>}}
  }
  InteractDetails {
   widgetClassPath=&CommonWidgetSwitcherLegacy
   bIsVariable=true
   parent=VerticalBox_89
   widget=InteractDetails
   properties {activeWidgetIndex=3}
   slot {class=VerticalBoxSlot}
  }
  CanInteract {
   widgetClassPath=&Spacer
   parent=InteractDetails
   widget=CanInteract
   properties {visibility=Collapsed}
   slot {class=WidgetSwitcherSlot}
  }
  PartyMemberText {
   widgetClassPath=&CommonTextBlock
   parent=InteractDetails
   widget=PartyMemberText
   properties {
    justification=Center
    text=Only Party Leader can Interact
    visibility=HitTestInvisible
    colorAndOpacity {colorUseRule=UseColor_Specified, specifiedColor=<rgba 1 0.0469 0 1>}
    font {
     bForceMonospaced=false
     bMaterialIsStencil=false
     fontMaterial=null
     fontObject=&BurbankSmall
     letterSpacing=0
     monospacedWidth=1
     size=16
     skewAmount=0
     typefaceFontName=Black
     outlineSettings {
      bApplyOutlineToDropShadows=false
      bMiteredCorners=false
      bSeparateFillAlpha=false
      outlineMaterial=null
      outlineSize=1
      outlineColor=<rgba 0 0.066 0.066 0.3>
     }
    }
    renderTransform {angle=0, scale=<1 1>, shear=<-8 0>, translation=<0 0>}
    shadowColorAndOpacity=<rgba 0.0052 0.0052 0.0056 0.822>
    shadowOffset=<0 0>
   }
   slot {class=WidgetSwitcherSlot, properties {horizontalAlignment=HAlign_Left, verticalAlignment=VAlign_Bottom, padding=<ltrb 4 0 0 0>}}
  }
  PartyTooLarge {
   widgetClassPath=&CommonTextBlock
   parent=InteractDetails
   widget=PartyTooLarge
   properties {
    justification=Center
    text=Your Party is too large for this Island
    visibility=HitTestInvisible
    colorAndOpacity {colorUseRule=UseColor_Specified, specifiedColor=<rgba 1 0.0469 0 1>}
    font {
     bForceMonospaced=false
     bMaterialIsStencil=false
     fontMaterial=null
     fontObject=&BurbankSmall
     letterSpacing=0
     monospacedWidth=1
     size=16
     skewAmount=0
     typefaceFontName=Black
     outlineSettings {
      bApplyOutlineToDropShadows=false
      bMiteredCorners=false
      bSeparateFillAlpha=false
      outlineMaterial=null
      outlineSize=1
      outlineColor=<rgba 0 0.066 0.066 0.3>
     }
    }
    renderTransform {angle=0, scale=<1 1>, shear=<-8 0>, translation=<0 0>}
    shadowColorAndOpacity=<rgba 0.0052 0.0052 0.0056 0.822>
    shadowOffset=<0 0>
   }
   slot {class=WidgetSwitcherSlot, properties {horizontalAlignment=HAlign_Left, verticalAlignment=VAlign_Bottom, padding=<ltrb 4 0 0 0>}}
  }
  Override {widgetClassPath=&HorizontalBox, parent=InteractDetails, widget=Override, slot {class=WidgetSwitcherSlot}}
  Text_Override {
   widgetClassPath=&CommonTextBlock
   parent=Override
   widget=Text_Override
   properties {
    justification=Center
    text=Interact to set code
    colorAndOpacity {colorUseRule=UseColor_Specified, specifiedColor=<rgba 1 0.5253 0 1>}
    font {
     bForceMonospaced=false
     bMaterialIsStencil=false
     fontMaterial=null
     fontObject=&BurbankSmall
     letterSpacing=0
     monospacedWidth=1
     size=16
     skewAmount=0
     typefaceFontName=Black
     outlineSettings {
      bApplyOutlineToDropShadows=false
      bMiteredCorners=false
      bSeparateFillAlpha=false
      outlineMaterial=null
      outlineSize=1
      outlineColor=<rgba 0 0.066 0.066 0.3>
     }
    }
    renderTransform {angle=0, scale=<1 1>, shear=<-8 0>, translation=<0 0>}
    shadowColorAndOpacity=<rgba 0.0052 0.0052 0.0056 0.822>
    shadowOffset=<0 0>
   }
   slot {class=HorizontalBoxSlot, properties {horizontalAlignment=HAlign_Left, verticalAlignment=VAlign_Bottom, padding=<ltrb 4 0 0 0>}}
  }
  Override_Locked_Timed {widgetClassPath=&HorizontalBox, parent=InteractDetails, widget=Override_Locked_Timed, slot {class=WidgetSwitcherSlot}}
  Text_Override_1 {
   widgetClassPath=&CommonTextBlock
   parent=Override_Locked_Timed
   widget=Text_Override_1
   properties {
    justification=Center
    text="Code Override Locked for "
    colorAndOpacity {colorUseRule=UseColor_Specified, specifiedColor=<rgba 1 0.5253 0 1>}
    font {
     bForceMonospaced=false
     bMaterialIsStencil=false
     fontMaterial=null
     fontObject=&BurbankSmall
     letterSpacing=0
     monospacedWidth=1
     size=16
     skewAmount=0
     typefaceFontName=Black
     outlineSettings {
      bApplyOutlineToDropShadows=false
      bMiteredCorners=false
      bSeparateFillAlpha=false
      outlineMaterial=null
      outlineSize=1
      outlineColor=<rgba 0 0.066 0.066 0.3>
     }
    }
    renderTransform {angle=0, scale=<1 1>, shear=<-8 0>, translation=<0 0>}
    shadowColorAndOpacity=<rgba 0.0052 0.0052 0.0056 0.822>
    shadowOffset=<0 0>
   }
   slot {class=HorizontalBoxSlot, properties {horizontalAlignment=HAlign_Left, verticalAlignment=VAlign_Bottom, padding=<ltrb 4 0 0 0>}}
  }
  Timer {
   widgetClassPath=&CommonTextBlock
   bIsVariable=true
   parent=Override_Locked_Timed
   widget=Timer
   properties {
    justification=Center
    text=10s
    colorAndOpacity {colorUseRule=UseColor_Specified, specifiedColor=<rgba 1 0.9534 0.974 1>}
    font {
     bForceMonospaced=false
     bMaterialIsStencil=false
     fontMaterial=null
     fontObject=&BurbankSmall
     letterSpacing=0
     monospacedWidth=1
     size=16
     skewAmount=0
     typefaceFontName=Black
     outlineSettings {
      bApplyOutlineToDropShadows=false
      bMiteredCorners=false
      bSeparateFillAlpha=false
      outlineMaterial=null
      outlineSize=1
      outlineColor=<rgba 0 0.066 0.066 0.3>
     }
    }
    renderTransform {angle=0, scale=<1 1>, shear=<-8 0>, translation=<0 0>}
    shadowColorAndOpacity=<rgba 0.0052 0.0052 0.0056 0.822>
    shadowOffset=<0 0>
   }
   slot {class=HorizontalBoxSlot, properties {horizontalAlignment=HAlign_Left, verticalAlignment=VAlign_Bottom, padding=<ltrb 4 0 0 0>}}
  }
 ]
}
