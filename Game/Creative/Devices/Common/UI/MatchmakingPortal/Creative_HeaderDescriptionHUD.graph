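# Matchmaking portal HUD callout that shows a header, text description or island details, and mobile matchmaking prompts, with a pointer to a cached world position.
# InitFromObject takes desired visibility from its HUD component and applies pixel offsets through spacers.
# Visibility changes play intro animations forward or in reverse, then collapse the widget.
# It hides on respawn start, and a request during a loading screen is stored and restored afterward.
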WidgetBlueprint /Game/Creative/Devices/Common/UI/MatchmakingPortal/Creative_HeaderDescriptionHUD
animations [
 intro2 {
  binding_count=1
  length_seconds=0.15
  object=&self:intro2
  sequence {bindings [HeaderText {
     id=#1
     possessed_class=CommonTextBlock
     tracks [
      {
       type=MovieScene2DTransformTrack
       sections [{
         type=MovieScene2DTransformSection
         end_frame=3
         start_frame=0
         channels [
          Translation.X {type=MovieSceneScriptingFloatChannel, keys [[0,50],[3,0]]}
          Translation.Y {type=MovieSceneScriptingFloatChannel, const [0,2]}
          Angle {type=MovieSceneScriptingFloatChannel}
          Scale.X {type=MovieSceneScriptingFloatChannel}
          Scale.Y {type=MovieSceneScriptingFloatChannel}
          Shear.X {type=MovieSceneScriptingFloatChannel}
          Shear.Y {type=MovieSceneScriptingFloatChannel}
         ]
        }]
      }
      {
       type=MovieSceneFloatTrack
       display_name=Render Opacity
       sections [{type=MovieSceneFloatSection, end_frame=3, start_frame=0, channels [None {type=MovieSceneScriptingFloatChannel, keys [[0,0],[3,1]]}]}]
      }
     ]
    }]}
 }
 DescriptionReveal {
  binding_count=1
  length_seconds=0.25
  object=&self:DescriptionReveal
  sequence {bindings [Text_Description {
     id=#2
     possessed_class=CommonTextBlock
     tracks [
      {
       type=MovieScene2DTransformTrack
       sections [{
         type=MovieScene2DTransformSection
         end_frame=5
         start_frame=0
         channels [
          Translation.X {type=MovieSceneScriptingFloatChannel}
          Translation.Y {type=MovieSceneScriptingFloatChannel}
          Angle {type=MovieSceneScriptingFloatChannel}
          Scale.X {type=MovieSceneScriptingFloatChannel, keys [[0,1.2424],[5,1]]}
          Scale.Y {type=MovieSceneScriptingFloatChannel, keys [[0,1.2424],[5,1]]}
          Shear.X {type=MovieSceneScriptingFloatChannel}
          Shear.Y {type=MovieSceneScriptingFloatChannel}
         ]
        }]
      }
      {
       type=MovieSceneFloatTrack
       display_name=Render Opacity
       sections [{type=MovieSceneFloatSection, end_frame=5, start_frame=0, channels [None {type=MovieSceneScriptingFloatChannel, keys [[0,0],[3,1],[5,1]]}]}]
      }
     ]
    }]}
 }
 Intro3 {
  binding_count=2
  length_seconds=3.75
  object=&self:Intro3
  sequence {bindings [
    Image_Backplate {
     id=#3
     possessed_class=Image
     tracks [{
       type=MovieSceneFloatTrack
       display_name=Render Opacity
       sections [{type=MovieSceneFloatSection, end_frame=75, start_frame=0, channels [None {type=MovieSceneScriptingFloatChannel, keys [[0,0],[30,0],[40,1]]}]}]
      }]
    }
    Image_Glow_1 {
     id=#4
     possessed_class=Image
     tracks [{
       type=MovieSceneFloatTrack
       display_name=Render Opacity
       sections [{type=MovieSceneFloatSection, end_frame=40, start_frame=0, channels [None {type=MovieSceneScriptingFloatChannel, keys [[0,1],[30,1],[40,0]]}]}]
      }]
    }
   ]}
 }
]
graphs [
 EventGraph {nodes [
   n1 = HandleLoadScreenChanged!() -> (OutputDelegate, then, PlayerController:Fort Player Controller Athena Object Reference, bEnableLoadScreen:bool=~, HUDReasonText:Text) {is=title K2Node_CustomEvent}
   > n2 = Branch(Condition=@n1.bEnableLoadScreen)
   > UpdateLoadingVisibility!(execute, self, bVisible:bool=false) -> (then)
   UpdateLoadingVisibility(execute=@n2.else, bVisible=Get BeforeLoadingScreenVisibility() -> (BeforeLoadingScreenVisibility:bool))
   n3 = Hide on Respawn Start!() -> (OutputDelegate, then, PlayerController:Fort Player Controller Zone Object Reference) {is=title K2Node_CustomEvent}
   > Update Visibility!(execute, self, bNewVis:bool=false) -> (then)
   n4 = UpdateLoadingVisibility!() -> (OutputDelegate, then, bVisible:bool=~) {is=title K2Node_CustomEvent}
   > UpdateAnimation!(execute, self, bVisible:bool=@n4.bVisible) -> (then)
   n5 = UpdateAnimation!() -> (OutputDelegate, then, bVisible:bool=~) {is=title K2Node_CustomEvent}
   > n6 = Branch(Condition=@n5.bVisible)
   > Call On Header About to be Displayed!(execute:Exec, self:Creative Header Description HUD Object Reference) -> (then:Exec) {is=title K2Node_CallDelegate}
   n7 = Get DescriptionReveal() -> (DescriptionReveal:Widget Animation Object Reference)
   > PlayAnimation(InAnimation=@n7)
   n8 = Get Intro3() -> (Intro3:Widget Animation Object Reference)
   > PlayAnimation(InAnimation=@n8)
   n9 = Get intro2() -> (intro2:Widget Animation Object Reference)
   > PlayAnimation(InAnimation=@n9)
   > SetVisibility(InVisibility=HitTestInvisible)
   > SetCachedWorldPos!(execute, self:Fort Portal HUDDetails Object Reference, InWorldPos:Vector=Get World Pointer Location() -> (World Pointer Location:Vector)) -> (then)
   PlayAnimationReverse(execute=@n6.else, InAnimation=@n7)
   > PlayAnimationReverse(InAnimation=@n8)
   > PlayAnimationReverse(InAnimation=@n9)
   > Delay(Duration=Get Animation Length() -> (Animation Length:double))
   > SetVisibility(InVisibility=Collapsed)
   > ClearCachedWorldPos(self:Fort Portal HUDDetails Object Reference)
   > Call On Header Hidden!(execute:Exec, self:Creative Header Description HUD Object Reference) -> (then:Exec) {is=title K2Node_CallDelegate}
   n10 = Update Visibility!() -> (OutputDelegate, then, bNewVis:bool=~) {is=title K2Node_CustomEvent}
   > n11 = Branch(Condition=Get FirstLoadingScreenEventFlagged() -> (FirstLoadingScreenEventFlagged:bool))
   n12 = Set BeforeLoadingScreenVisibility(execute=[@n14.else, @n11.then], BeforeLoadingScreenVisibility:bool=@n10.bNewVis) -> (Output_Get:bool)
   > UpdateAnimation(bVisible=@n12) {is=title K2Node_CallFunction}
   n13 = Cast To FortPlayerControllerAthena(execute=@n11.else, Object=GetOwningPlayer())
   > n14 = Branch(Condition=IsLoadingScreenActive!(self:Fort Player Controller Athena Object Reference=@n13) -> (ReturnValue:bool))
   > Set BeforeLoadingScreenVisibility(BeforeLoadingScreenVisibility=@n10.bNewVis)
   > Set FirstLoadingScreenEventFlagged(FirstLoadingScreenEventFlagged:bool=true) -> (Output_Get:bool)
   > UpdateAnimation(bVisible=false)
   Event Construct()
   > Parent: Construct()
   > n15 = Cast To FortPlayerControllerAthena(Object=GetOwningPlayer())
   > Bind Event to On Started Respawn(self:Fort Player Controller Zone Object Reference=@n15, Delegate=@n3) {is=title K2Node_AddDelegate}
   > Bind Event to In Game Load Screen Changed(self:Fort Player Controller Athena Object Reference=@n15, Delegate=@n1) {is=title K2Node_AddDelegate}
   n16 = Event InitFromObject!() -> (OutputDelegate, then, InitObject:Object Reference) {is=title K2Node_Event}
   > n17 = Cast To Creative_HeaderDescriptionHUD_Component!(execute, Object=@n16.InitObject) -> (then, CastFailed, AsCreative Header Description HUD Component:Creative Header Description HUD Component Object Reference) {is=title K2Node_DynamicCast}
   > Set HUDComponent(HUDComponent:Creative Header Description HUD Component Object Reference=@n17) -> (Output_Get:Creative Header Description HUD Component Object Reference)
   > Update Visibility(bNewVis=Get bDesiredVisibility!(self:Header Description HUDComponent Object Reference=Get HUDComponent() -> (HUDComponent:Creative Header Description HUD Component Object Reference)) -> (bDesiredVisibility:bool)) {is=title K2Node_CallFunction}
   > Set Offset!(execute, self, Offset:Vector 2D Structure=Get PixelOffset() -> (PixelOffset:Vector 2D Structure)) -> (then) {is=title K2Node_CallFunction}
   > n18 = Branch(Condition=Get bHUDAlreadyExists!(self:Header Description HUDComponent Object Reference=Get HUDComponent()) -> (bHUDAlreadyExists:bool))
   Bind Event to On Header About to be Displayed(execute=@n18.else, self:Creative Header Description HUD Object Reference, Delegate=Create Event(self=Get HUDComponent())) {is=title K2Node_AddDelegate}
   > Bind Event to On Header Hidden(self:Creative Header Description HUD Object Reference, Delegate=Create Event(self=Get HUDComponent())) {is=title K2Node_AddDelegate}
   Event PreConstruct()
   > SetLineMeshMaterial!(
    execute
    self:Athena Marker Pointer Object Reference=Get MarkerPointer() -> (MarkerPointer:Athena Marker Pointer Object Reference)
    Material:Material Interface Object Reference=/Game/Athena/HUD/Art/Materials/M_UI_ModalRadialExpand_PointerVersion
   ) -> (then)
  ]}
 HandleLoadingScreenChanged {nodes [HandleLoadingScreenChanged!() -> (then, PlayerController:Fort Player Controller Athena Object Reference, Loading:bool=~, Reason:Text) {is=title K2Node_FunctionEntry}]}
 OnHeaderAboutToBeDisplayed {nodes [OnHeaderAboutToBeDisplayed() {is=title K2Node_FunctionEntry}]}
 OnHeaderHidden {nodes [OnHeaderHidden() {is=title K2Node_FunctionEntry}]}
 Set Header Details {nodes [
   n1 = Set Header Details!() -> (then, Header:Text, World Pointer Location:Vector=~, LinkId:Online Link Id Structure) {is=title K2Node_FunctionEntry}
   n2 = Get HeaderText() -> (HeaderText:Common Text Object Reference)
   > "SetText (Text)"(self=@n2, InText=@n1)
   > QueryGameActivity!(execute, self:Fort Portal HUDDetails Object Reference, LinkId:"Online Link Id Structure (by ref)"=@n1.LinkId) -> (then)
   > SetVisibility(self=@n2, InVisibility=Select(Option 0=SelfHitTestInvisible, Option 1=Hidden, Index=TextIsEmpty(InText=@n1)))
   > n3 = Set World Pointer Location(World Pointer Location:Vector=@n1.World Pointer Location) -> (Output_Get:Vector)
   > SetCachedWorldPos(InWorldPos=@n3)
  ]}
 Set Offset {nodes [
   n1 = Set Offset!() -> (then, Offset_X:double, Offset_Y:double) {is=title K2Node_FunctionEntry}
   > SetSize!(
    execute
    self:Spacer Object Reference=Get VerticalOffset() -> (VerticalOffset:Spacer Object Reference)
    InSize_X:double=0
    InSize_Y:double=@n1.Offset_Y
   ) -> (then)
   > SetSize(self=Get HorizontalOffset() -> (HorizontalOffset:Spacer Object Reference), InSize_X=@n1, InSize_Y=0)
  ]}
 SetDescription {nodes [
   n1 = SetDescription!() -> (then, Owning Object:Object Reference) {is=title K2Node_FunctionEntry}
   > n2 = Get Description Details!(execute:Exec, self:Object Reference=@n1) -> (then:Exec, Description:Text) {is=title K2Node_Message}
   n3 = Get Text_Description() -> (Text_Description:Common Text Object Reference)
   > "SetText (Text)"(self=@n3, InText=@n2)
   n4 = TextIsEmpty(InText=@n2)
   > SetVisibility(self=@n3, InVisibility=Select(Option 0=SelfHitTestInvisible, Option 1=Collapsed, Index=@n4))
   > Branch(Condition=NOT Boolean(A=@n4))
   > Switch Detail Panel!(execute, self, New State:ECreativeHUDWidgets Enum=NewEnumerator0) -> (then)
  ]}
 Switch Detail Panel {nodes [
   n1 = Switch Detail Panel!() -> (then, New State:ECreativeHUDWidgets Enum=NewEnumerator0) {is=title K2Node_FunctionEntry}
   > SetActiveWidgetIndex(
    self=Get DetailsSwitcher() -> (DetailsSwitcher:Common Widget Switcher Legacy Object Reference)
    Index="To Integer (Byte)"!(InByte:Byte=@n1) -> (ReturnValue:int=0)
   )
  ]}
 Update IslandPortalDetails {nodes [
   n1 = Update IslandPortalDetails!() -> (then, Owning Object:Object Reference) {is=title K2Node_FunctionEntry}
   > Update Island Settings!(
    execute
    self:Creative Island Details Object Reference=Get CreativeIslandDetails() -> (CreativeIslandDetails:Creative Island Details Object Reference)
    Owning Object:Object Reference=@n1
   ) -> (then)
   > Switch Detail Panel(New State=NewEnumerator1) {is=title K2Node_CallFunction}
  ]}
 Update World Pointer Location {nodes [
   n1 = Update World Pointer Location!() -> (then, World Pointer Location:Vector=~) {is=title K2Node_FunctionEntry}
   > n2 = Set World Pointer Location(World Pointer Location=@n1)
   > SetCachedWorldPos(InWorldPos=@n2)
  ]}
]
variables [
 Animation Length {type=number, default=0.25}
 BeforeLoadingScreenVisibility {type=boolean}
 FirstLoadingScreenEventFlagged {type=boolean}
 HUDComponent {type=Creative_HeaderDescriptionHUD_Component_C, default=null}
 OnHeaderAboutToBeDisplayed {}
 OnHeaderHidden {}
 PixelOffset {type=Vector2D, default=<100 150>}
 World Pointer Location {type=Vector}
]
widgets {
 info {inheritedWidgetCount=3, parentClass=/Script/FortniteUI.FortPortalHUDDetails, rootWidgetClass=&Overlay, widgetCount=24}
 widgets [
  Overlay_2 {widgetClassPath=&Overlay, widget=Overlay_2}
  HOffset {widgetClassPath=&HorizontalBox, parent=Overlay_2, widget=HOffset, slot {class=OverlaySlot}}
  HorizontalOffset {
   widgetClassPath=&Spacer
   bIsVariable=true
   parent=HOffset
   widget=HorizontalOffset
   properties {size=<0 0>}
   slot {class=HorizontalBoxSlot}
  }
  VOffset {
   widgetClassPath=&VerticalBox
   bIsVariable=true
   parent=HOffset
   widget=VOffset
   slot {class=HorizontalBoxSlot, properties {horizontalAlignment=HAlign_Left, verticalAlignment=VAlign_Top}}
  }
  Overlay_167 {
   widgetClassPath=&Overlay
   parent=VOffset
   widget=Overlay_167
   slot {class=VerticalBoxSlot, properties {horizontalAlignment=HAlign_Left, verticalAlignment=VAlign_Top}}
  }
  Image_Backplate {
   widgetClassPath=&Image
   bIsVariable=true
   parent=Overlay_167
   widget=Image_Backplate
   properties {
    renderOpacity=0
    brush {
     bIsDynamicallyLoaded=false
     drawAs=Image
     imageType=NoImage
     mirroring=NoMirror
     resourceName=None
     resourceObject=/Game/Playgrounds/Items/Portals/UI/Art/M_SimpleSkewBox_Cr_PortalCalloutBacking
     tiling=NoTile
     imageSize=<32 32>
     margin=<ltrb 0 0 0 0>
     outlineSettings {
      bUseBrushTransparency=false
      roundingType=HalfHeightRadius
      width=0
      color {colorUseRule=UseColor_Specified, specifiedColor=<rgba 0 0 0 0>}
      cornerRadii=<0 0 0 0>
     }
     tintColor {colorUseRule=UseColor_Specified, specifiedColor=<rgba 1 1 1 1>}
     uVRegion {bIsValid=false, max=<0 0>, min=<0 0>}
    }
    colorAndOpacity=<rgba 0 0.0062 0.0677 0.9>
    renderTransform {angle=0, scale=<1.45 1.5>, shear=<0 0>, translation=<0 0>}
   }
   slot {class=OverlaySlot, properties {horizontalAlignment=HAlign_Fill, verticalAlignment=VAlign_Fill}}
  }
  Image_Glow_1 {
   widgetClassPath=&Image
   parent=Overlay_167
   widget=Image_Glow_1
   properties {
    renderOpacity=0
    visibility=SelfHitTestInvisible
    brush {
     bIsDynamicallyLoaded=false
     drawAs=Image
     imageType=NoImage
     mirroring=NoMirror
     resourceName=None
     resourceObject=/Game/Athena/HUD/Materials/UI_RebootAlert_Glow_Inst
     tiling=NoTile
     imageSize=<32 32>
     margin=<ltrb 0 0 0 0>
     outlineSettings {
      bUseBrushTransparency=false
      roundingType=HalfHeightRadius
      width=0
      color {colorUseRule=UseColor_Specified, specifiedColor=<rgba 0 0 0 0>}
      cornerRadii=<0 0 0 0>
     }
     tintColor {colorUseRule=UseColor_Specified, specifiedColor=<rgba 1 1 1 1>}
     uVRegion {bIsValid=false, max=<0 0>, min=<0 0>}
    }
    colorAndOpacity=<rgba 0 0.0138 0.151 1>
    renderTransform {angle=0, scale=<2.5254 2.4649>, shear=<0 0>, translation=<0 0>}
   }
   slot {class=OverlaySlot, properties {horizontalAlignment=HAlign_Fill, verticalAlignment=VAlign_Fill}}
  }
  VerticalBox_0 {
   widgetClassPath=&VerticalBox
   parent=Overlay_167
   widget=VerticalBox_0
   slot {class=OverlaySlot, properties {horizontalAlignment=HAlign_Fill, verticalAlignment=VAlign_Fill}}
  }
  HeaderText {
   widgetClassPath=&CommonTextBlock
   bIsVariable=true
   parent=VerticalBox_0
   widget=HeaderText
   properties {
    autoWrapText=true
    bAutoCollapseWithEmptyText=true
    clipping=ClipToBounds
    text=Really Really Long Text STring Fun and Games
    textTransformPolicy=ToUpper
    wrapTextAt=465
    wrappingPolicy=AllowPerCharacterWrapping
    font {
     bForceMonospaced=false
     bMaterialIsStencil=false
     fontMaterial=null
     fontObject=&BurbankSmall
     letterSpacing=0
     monospacedWidth=1
     size=24
     skewAmount=0
     typefaceFontName=Black
     outlineSettings {
      bApplyOutlineToDropShadows=false
      bMiteredCorners=false
      bSeparateFillAlpha=false
      outlineMaterial=null
      outlineSize=1
      outlineColor=<rgba 0.0052 0.0052 0.0056 0.2>
     }
    }
    renderTransform {angle=0, scale=<1 1>, shear=<-8 0>, translation=<0 0>}
    shadowColorAndOpacity=<rgba 0.0052 0.0052 0.0056 0.822>
    shadowOffset=<0 0>
   }
   slot {class=VerticalBoxSlot, properties {verticalAlignment=VAlign_Top, padding=<ltrb 12 5 0 0>}}
  }
  SqueegeeInjectionSlot {
   widgetClassPath=/Game/UI/Plugin/Squeegee/SqueegeeInjectionSlot
   bIsVariable=true
   parent=VerticalBox_0
   widget=SqueegeeInjectionSlot
   slot {class=VerticalBoxSlot}
  }
  DetailsSwitcher {
   widgetClassPath=&CommonWidgetSwitcherLegacy
   bIsVariable=true
   parent=VerticalBox_0
   widget=DetailsSwitcher
   properties {activeWidgetIndex=1}
   slot {class=VerticalBoxSlot}
  }
  Text_Description {
   widgetClassPath=&CommonTextBlock
   bIsVariable=true
   parent=DetailsSwitcher
   widget=Text_Description
   properties {
    text=Suspiciously long vague description text about this that or the other WHO KNOWS
    wrapTextAt=465
    colorAndOpacity {colorUseRule=UseColor_Specified, specifiedColor=<rgba 0 1 1 1>}
    font {
     bForceMonospaced=false
     bMaterialIsStencil=false
     fontMaterial=null
     fontObject=&BurbankSmall
     letterSpacing=0
     monospacedWidth=1
     size=18
     skewAmount=0
     typefaceFontName=Bold
     outlineSettings {
      bApplyOutlineToDropShadows=false
      bMiteredCorners=false
      bSeparateFillAlpha=false
      outlineMaterial=null
      outlineSize=1
      outlineColor=<rgba 0.0052 0.0052 0.0056 0.149>
     }
    }
    shadowColorAndOpacity=<rgba 0.0052 0.0052 0.0056 0.822>
    shadowOffset=<0 0>
   }
   slot {class=WidgetSwitcherSlot, properties {horizontalAlignment=HAlign_Left, verticalAlignment=VAlign_Top, padding=<ltrb 12 0 0 0>}}
  }
  CreativeIslandDetails {
   widgetClassPath=/Game/Creative/Devices/MatchmakingPortal/CreativeIslandDetails
   bIsVariable=true
   parent=DetailsSwitcher
   widget=CreativeIslandDetails
   properties {visibility=Collapsed}
   slot {class=WidgetSwitcherSlot, properties {horizontalAlignment=HAlign_Left, verticalAlignment=VAlign_Top, padding=<ltrb 12 0 0 0>}}
  }
  Border_109 {
   widgetClassPath=&Border
   parent=VerticalBox_0
   widget=Border_109
   properties {
    background {
     bIsDynamicallyLoaded=false
     drawAs=Image
     imageType=NoImage
     mirroring=NoMirror
     resourceName=None
     resourceObject=null
     tiling=NoTile
     imageSize=<32 32>
     margin=<ltrb 0 0 0 0>
     outlineSettings {
      bUseBrushTransparency=false
      roundingType=HalfHeightRadius
      width=0
      color {colorUseRule=UseColor_Specified, specifiedColor=<rgba 0 0 0 0>}
      cornerRadii=<0 0 0 0>
     }
     tintColor {colorUseRule=UseColor_Specified, specifiedColor=<rgba 0 1 1 1>}
     uVRegion {bIsValid=false, max=<0 0>, min=<0 0>}
    }
    padding=<ltrb 4 4 4 2>
   }
   slot {class=VerticalBoxSlot, properties {padding=<ltrb 0 6 0 0>}}
  }
  MarkerPointer {
   widgetClassPath=/Script/FortniteUI.AthenaMarkerPointer
   bInherited=true
   bIsVariable=true
   parent=Border_109
   widget=MarkerPointer
   slot {class=BorderSlot, properties {padding=<ltrb 4 4 4 2>}}
  }
  MobileInfo_Matchmaking_Switcher {
   widgetClassPath=&CommonActivatableWidgetSwitcher
   bInherited=true
   bIsVariable=true
   parent=VerticalBox_0
   widget=MobileInfo_Matchmaking_Switcher
   slot {class=VerticalBoxSlot}
  }
  Overlay_0 {widgetClassPath=&Overlay, parent=MobileInfo_Matchmaking_Switcher, widget=Overlay_0, slot {class=WidgetSwitcherSlot}}
  Image_Backplate_1 {
   widgetClassPath=&Image
   parent=Overlay_0
   widget=Image_Backplate_1
   properties {
    renderOpacity=100
    brush {
     bIsDynamicallyLoaded=false
     drawAs=Image
     imageType=NoImage
     mirroring=NoMirror
     resourceName=None
     resourceObject=/Game/Playgrounds/Items/Portals/UI/Art/M_SimpleSkewBox_Cr_PortalMobileMatchmakingBacking_Inst
     tiling=NoTile
     imageSize=<32 32>
     margin=<ltrb 0 0 0 0>
     outlineSettings {
      bUseBrushTransparency=false
      roundingType=HalfHeightRadius
      width=0
      color {colorUseRule=UseColor_Specified, specifiedColor=<rgba 0 0 0 0>}
      cornerRadii=<0 0 0 0>
     }
     tintColor {colorUseRule=UseColor_Specified, specifiedColor=<rgba 1 1 1 1>}
     uVRegion {bIsValid=false, max=<0 0>, min=<0 0>}
    }
    colorAndOpacity=<rgba 0 0.051 0.6863 0.8392>
    renderTransform {angle=0, scale=<1.45 1.5>, shear=<0 0>, translation=<0 0>}
   }
   slot {class=OverlaySlot, properties {horizontalAlignment=HAlign_Fill, verticalAlignment=VAlign_Fill}}
  }
  Text_Enter_Matchmaking {
   widgetClassPath=&CommonTextBlock
   parent=Overlay_0
   widget=Text_Enter_Matchmaking
   properties {
    text=Enter Matchmaking
    wrapTextAt=465
    colorAndOpacity {colorUseRule=UseColor_Specified, specifiedColor=<rgba 0 1 1 1>}
    font {
     bForceMonospaced=false
     bMaterialIsStencil=false
     fontMaterial=null
     fontObject=&BurbankSmall
     letterSpacing=0
     monospacedWidth=1
     size=15
     skewAmount=0
     typefaceFontName=Black
     outlineSettings {
      bApplyOutlineToDropShadows=false
      bMiteredCorners=false
      bSeparateFillAlpha=false
      outlineMaterial=null
      outlineSize=1
      outlineColor=<rgba 0.0052 0.0052 0.0056 0.149>
     }
    }
    renderTransform {angle=0, scale=<1 1>, shear=<-12 0>, translation=<0 0>}
    shadowColorAndOpacity=<rgba 0.0052 0.0052 0.0056 0.822>
    shadowOffset=<0 0>
   }
   slot {class=OverlaySlot, properties {horizontalAlignment=HAlign_Right, verticalAlignment=VAlign_Fill, padding=<ltrb 0 5 0 0>}}
  }
  Overlay_1 {widgetClassPath=&Overlay, parent=MobileInfo_Matchmaking_Switcher, widget=Overlay_1, slot {class=WidgetSwitcherSlot}}
  Image_Backplate_2 {
   widgetClassPath=&Image
   parent=Overlay_1
   widget=Image_Backplate_2
   properties {
    renderOpacity=100
    brush {
     bIsDynamicallyLoaded=false
     drawAs=Image
     imageType=NoImage
     mirroring=NoMirror
     resourceName=None
     resourceObject=&M_SimpleSkewBox_Cr_PortalMobileMatchmakingBacking_Inst
     tiling=NoTile
     imageSize=<32 32>
     margin=<ltrb 0 0 0 0>
     outlineSettings {
      bUseBrushTransparency=false
      roundingType=HalfHeightRadius
      width=0
      color {colorUseRule=UseColor_Specified, specifiedColor=<rgba 0 0 0 0>}
      cornerRadii=<0 0 0 0>
     }
     tintColor {colorUseRule=UseColor_Specified, specifiedColor=<rgba 1 1 1 1>}
     uVRegion {bIsValid=false, max=<0 0>, min=<0 0>}
    }
    colorAndOpacity=<rgba 0 0.051 0.6863 0.8392>
    renderTransform {angle=0, scale=<1.45 1.5>, shear=<0 0>, translation=<0 0>}
   }
   slot {class=OverlaySlot, properties {horizontalAlignment=HAlign_Fill, verticalAlignment=VAlign_Fill}}
  }
  Text_Cancel_Matchmaking {
   widgetClassPath=&CommonTextBlock
   parent=Overlay_1
   widget=Text_Cancel_Matchmaking
   properties {
    text=Cancel Matchmaking
    wrapTextAt=465
    colorAndOpacity {colorUseRule=UseColor_Specified, specifiedColor=<rgba 0 1 1 1>}
    font {
     bForceMonospaced=false
     bMaterialIsStencil=false
     fontMaterial=null
     fontObject=&BurbankSmall
     letterSpacing=0
     monospacedWidth=1
     size=15
     skewAmount=0
     typefaceFontName=Black
     outlineSettings {
      bApplyOutlineToDropShadows=false
      bMiteredCorners=false
      bSeparateFillAlpha=false
      outlineMaterial=null
      outlineSize=1
      outlineColor=<rgba 0.0052 0.0052 0.0056 0.149>
     }
    }
    renderTransform {angle=0, scale=<1 1>, shear=<-12 0>, translation=<0 0>}
    shadowColorAndOpacity=<rgba 0.0052 0.0052 0.0056 0.822>
    shadowOffset=<0 0>
   }
   slot {class=OverlaySlot, properties {horizontalAlignment=HAlign_Right, padding=<ltrb 0 5 0 0>}}
  }
  VerticalOffset {
   widgetClassPath=&Spacer
   bIsVariable=true
   parent=VOffset
   widget=VerticalOffset
   properties {size=<0 0>}
   slot {class=VerticalBoxSlot}
  }
  TutorialHighlight {widgetClassPath=&TutorialHighlight}
 ]
}
